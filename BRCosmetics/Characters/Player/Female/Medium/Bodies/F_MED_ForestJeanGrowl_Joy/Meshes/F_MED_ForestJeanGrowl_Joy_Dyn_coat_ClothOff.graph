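# Physics asset for the dynamic coat bones of the female medium ForestJeanGrowl_Joy body, apparently the ClothOff variant.
# Five capsule chains form the coat: two front chains hang from spine_03, three back chains from MSTR anchors on spine_01, with limited swing and twist.
# Angular-free cross-links join neighboring chains into a connected skirt.
# Pelvis, spine and leg capsules plus a large dyn_ground sphere supply body and floor shapes.
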PhysicsAsset /BRCosmetics/Characters/Player/Female/Medium/Bodies/F_MED_ForestJeanGrowl_Joy/Meshes/F_MED_ForestJeanGrowl_Joy_Dyn_coat_ClothOff
bodies {
 calf_l [{shapeType=Capsule, length=50.6561, radius=12.7157, shapeName=pelvis_capsule, center=<-30.4611 2.7534 5.3559>, rotation=<rot 90 0 0>}]
 calf_r [{shapeType=Capsule, length=50.6561, radius=12.8349, shapeName=pelvis_capsule, center=<30.4611 -2.7534 -5.4696>, rotation=<rot 90 180 0>}]
 dyn_coat_bk_01 [{shapeType=Capsule, length=19.2053, radius=20, shapeName=dyn_coat_l_01_capsule, center=<8.7723 0.0007 20>, rotation=<rot 90 0 0>}]
 dyn_coat_bk_02 [{shapeType=Capsule, length=8.4314, radius=17.6507, shapeName=dyn_coat_bk_02_capsule, center=<0 0 15>, rotation=<rot 0 -90 90.0002>}]
 dyn_coat_bk_04 [{shapeType=Capsule, length=11.0821, radius=15, shapeName=dyn_coat_bk_02_capsule, center=<0 0 12>, rotation=<rot -0.0001 0 90>}]
 dyn_coat_bk_05 [{shapeType=Capsule, length=13.9348, radius=15, shapeName=dyn_coat_bk_02_capsule, center=<0 0 12>, rotation=<rot -0.0001 0 90>}]
 dyn_coat_bk_06 [{shapeType=Capsule, length=24.4783, radius=15.9395, shapeName=dyn_coat_bk_02_capsule, center=<3.462 0 12>, rotation=<rot -0.0001 0 90>}]
 dyn_coat_bk_MSTR [{shapeType=Capsule, length=0.505, radius=0.505, shapeName=dyn_coat_l_MSTR_capsule}]
 dyn_coat_bk_l_01 [{shapeType=Capsule, length=25.9514, radius=18.9293, shapeName=dyn_coat_l_01_capsule, center=<10.7079 0 20>, rotation=<rot 90 0 0>}]
 dyn_coat_bk_l_02 [{shapeType=Capsule, length=21.2561, radius=20, shapeName=dyn_coat_l_01_capsule, center=<10.3356 0 20>, rotation=<rot 90 0 0>}]
 dyn_coat_bk_l_04 [{shapeType=Capsule, length=21.2561, radius=20, shapeName=dyn_coat_l_01_capsule, center=<10.3356 0 20>, rotation=<rot 90 0 0>}]
 dyn_coat_bk_l_05 [{shapeType=Capsule, length=16.3952, radius=20.2858, shapeName=dyn_coat_bk_r_05_capsule, center=<0 0 15>, rotation=<rot 0 -90.0001 -89.9999>}]
 dyn_coat_bk_l_06 [{
   shapeType=Capsule
   length=30.5134
   radius=15.9395
   shapeName=dyn_coat_bk_02_capsule
   center=<5.881 -0.8946 5>
   rotation=<rot -0.0004 -9.9998 95>
  }]
 dyn_coat_bk_l_MSTR [{shapeType=Capsule, length=0.505, radius=0.505, shapeName=dyn_coat_l_MSTR_capsule}]
 dyn_coat_bk_r_01 [{shapeType=Capsule, length=25.8014, radius=20, shapeName=dyn_coat_l_01_capsule, center=<10.7079 0 20>, rotation=<rot 90 0 0>}]
 dyn_coat_bk_r_02 [{shapeType=Capsule, length=21.2561, radius=20, shapeName=dyn_coat_l_01_capsule, center=<10.3356 0 20>, rotation=<rot 90 0 0>}]
 dyn_coat_bk_r_04 [{shapeType=Capsule, length=21.2561, radius=20, shapeName=dyn_coat_l_01_capsule, center=<10.3356 0 20>, rotation=<rot 90 0 0>}]
 dyn_coat_bk_r_05 [{shapeType=Capsule, length=16.3952, radius=20.1665, shapeName=dyn_coat_bk_r_05_capsule, center=<0 0 15>, rotation=<rot 0 -90.0001 -89.9999>}]
 dyn_coat_bk_r_06 [{shapeType=Capsule, length=24.4783, radius=15.9395, shapeName=dyn_coat_bk_02_capsule, center=<7.3484 1.0413 12>, rotation=<rot -0.0001 15 90>}]
 dyn_coat_bk_r_MSTR [{shapeType=Capsule, length=0.505, radius=0.505, shapeName=dyn_coat_l_MSTR_capsule}]
 dyn_coat_l_02 [{shapeType=Capsule, length=12.2561, radius=8, shapeName=dyn_coat_l_01_capsule, center=<5.44 0 5.5818>, rotation=<rot 90 0 0>}]
 dyn_coat_l_03 [{shapeType=Capsule, length=21.2561, radius=8, shapeName=dyn_coat_l_01_capsule, center=<10.3356 0 8>, rotation=<rot 90 0 0>}]
 dyn_coat_l_04 [{
   shapeType=Capsule
   length=12.5557
   radius=14.2351
   shapeName=dyn_coat_l_01_capsule
   center=<10.3356 0 15.4045>
   rotation=<rot 82.9334 45.1087 -45.1099>
  }]
 dyn_coat_l_05 [{shapeType=Capsule, length=10.1064, radius=15.8997, shapeName=dyn_coat_l_01_capsule, center=<3.5998 -5 14>, rotation=<rot 65 90.0001 -0.0001>}]
 dyn_coat_l_06 [{
   shapeType=Capsule
   length=25.285
   radius=14.7463
   shapeName=dyn_coat_bk_02_capsule
   center=<2.4088 -9 19.0099>
   rotation=<rot 2.4975 -4.3327 114.9055>
  }]
 dyn_coat_r_02 [{shapeType=Capsule, length=12.2561, radius=8, shapeName=dyn_coat_l_01_capsule, center=<5.44 0 5.5818>, rotation=<rot 90 0 0>}]
 dyn_coat_r_03 [{shapeType=Capsule, length=21.2561, radius=8, shapeName=dyn_coat_l_01_capsule, center=<10.3356 0 5.0234>, rotation=<rot 90 0 0>}]
 dyn_coat_r_04 [{shapeType=Capsule, length=21.2561, radius=15, shapeName=dyn_coat_l_01_capsule, center=<10.3356 0 15.4045>, rotation=<rot 90 0 0>}]
 dyn_coat_r_05 [{shapeType=Capsule, length=11.1277, radius=18.6193, shapeName=dyn_coat_l_01_capsule, center=<3.5998 0 16.4623>, rotation=<rot 90 0 0>}]
 dyn_coat_r_06 [{shapeType=Capsule, length=28.8174, radius=16.3956, shapeName=dyn_coat_bk_02_capsule, center=<3.462 0 12>, rotation=<rot -0.0001 -0.0001 50>}]
 dyn_ground [{shapeType=Sphere, radius=250, shapeName=dyn_ground_sphere, center=<0 -14.1596 -242.4057>}]
 pelvis [{shapeType=Capsule, length=5, radius=14.021, shapeName=pelvis_capsule, center=<-4.8171 -0.6518 -0.0005>}]
 spine_01 [{shapeType=Capsule, length=4.9652, radius=10.9987, shapeName=pelvis_capsule, center=<-0.9991 -0.3721 0>}]
 spine_03 [{shapeType=Capsule, length=4.4401, radius=10.4987, shapeName=pelvis_capsule, center=<2.1769 0.7985 0>}]
 thigh_l [{
   shapeType=Capsule
   length=29.589
   radius=11.0541
   shapeName=pelvis_capsule
   center=<-20.0609 -2.1741 1.8369>
   rotation=<rot 88.6498 0.0001 0.0002>
  }]
 thigh_r [{
   shapeType=Capsule
   length=29.589
   radius=11.0541
   shapeName=pelvis_capsule
   center=<20.0636 2.1741 -1.7232>
   rotation=<rot 88.6498 0.0001 -179.9998>
  }]
}
constraints [
 {bone1Name=dyn_coat_l_02, bone2Name=spine_03, swing1LimitDegrees=25, swing2LimitDegrees=45, twistLimitDegrees=5}
 {bone1Name=dyn_coat_l_03, bone2Name=dyn_coat_l_02, swing1LimitDegrees=15, swing2LimitDegrees=45, twistLimitDegrees=5}
 {bone1Name=dyn_coat_l_04, bone2Name=dyn_coat_l_03, swing1LimitDegrees=15, swing2LimitDegrees=45, twistLimitDegrees=5}
 {bone1Name=dyn_coat_l_05, bone2Name=dyn_coat_l_04, swing1LimitDegrees=10, swing2LimitDegrees=45, twistLimitDegrees=5}
 {bone1Name=dyn_coat_l_06, bone2Name=dyn_coat_l_05, swing1LimitDegrees=10, swing2LimitDegrees=45, twistLimitDegrees=5}
 {bone1Name=dyn_coat_r_02, bone2Name=spine_03, swing1LimitDegrees=25, swing2LimitDegrees=45, twistLimitDegrees=5}
 {bone1Name=dyn_coat_r_03, bone2Name=dyn_coat_r_02, swing1LimitDegrees=10, swing2LimitDegrees=45, twistLimitDegrees=5}
 {bone1Name=dyn_coat_r_04, bone2Name=dyn_coat_r_03, swing1LimitDegrees=10, swing2LimitDegrees=45, twistLimitDegrees=5}
 {bone1Name=dyn_coat_r_05, bone2Name=dyn_coat_r_04, swing1LimitDegrees=10, swing2LimitDegrees=45, twistLimitDegrees=5}
 {bone1Name=dyn_coat_r_06, bone2Name=dyn_coat_r_05, swing1LimitDegrees=10, swing2LimitDegrees=45, twistLimitDegrees=5}
 {bone1Name=dyn_coat_bk_01, bone2Name=dyn_coat_bk_MSTR, swing1LimitDegrees=10, swing2LimitDegrees=15, twistLimitDegrees=5}
 {bone1Name=dyn_coat_bk_02, bone2Name=dyn_coat_bk_01, swing1LimitDegrees=10, swing2LimitDegrees=15, twistLimitDegrees=5}
 {bone1Name=dyn_coat_bk_l_01, bone2Name=dyn_coat_bk_l_MSTR, swing1LimitDegrees=10, swing2LimitDegrees=45, twistLimitDegrees=5}
 {bone1Name=dyn_coat_bk_l_02, bone2Name=dyn_coat_bk_l_01, swing1LimitDegrees=10, swing2LimitDegrees=45, twistLimitDegrees=5}
 {bone1Name=dyn_coat_bk_r_01, bone2Name=dyn_coat_bk_r_MSTR, swing1LimitDegrees=10, swing2LimitDegrees=45, twistLimitDegrees=5}
 {bone1Name=dyn_coat_bk_r_02, bone2Name=dyn_coat_bk_r_01, swing1LimitDegrees=15, swing2LimitDegrees=45, twistLimitDegrees=5}
 {
  bone1Name=calf_l
  bone2Name=thigh_l
  swing1LimitDegrees=45
  swing1Motion=Free
  swing2LimitDegrees=45
  swing2Motion=Free
  twistLimitDegrees=45
  twistMotion=Free
 }
 {
  bone1Name=dyn_coat_bk_MSTR
  bone2Name=spine_01
  swing1LimitDegrees=45
  swing1Motion=Free
  swing2LimitDegrees=45
  swing2Motion=Free
  twistLimitDegrees=45
  twistMotion=Free
 }
 {
  bone1Name=dyn_coat_bk_l_MSTR
  bone2Name=spine_01
  swing1LimitDegrees=45
  swing1Motion=Free
  swing2LimitDegrees=45
  swing2Motion=Free
  twistLimitDegrees=45
  twistMotion=Free
 }
 {
  bone1Name=dyn_coat_bk_r_MSTR
  bone2Name=spine_01
  swing1LimitDegrees=45
  swing1Motion=Free
  swing2LimitDegrees=45
  swing2Motion=Free
  twistLimitDegrees=45
  twistMotion=Free
 }
 {
  bone1Name=thigh_l
  bone2Name=pelvis
  swing1LimitDegrees=45
  swing1Motion=Free
  swing2LimitDegrees=45
  swing2Motion=Free
  twistLimitDegrees=45
  twistMotion=Free
 }
 {
  bone1Name=spine_01
  bone2Name=pelvis
  swing1LimitDegrees=45
  swing1Motion=Free
  swing2LimitDegrees=45
  swing2Motion=Free
  twistLimitDegrees=45
  twistMotion=Free
 }
 {
  bone1Name=spine_03
  bone2Name=spine_01
  swing1LimitDegrees=45
  swing1Motion=Free
  swing2LimitDegrees=45
  swing2Motion=Free
  twistLimitDegrees=45
  twistMotion=Free
 }
 {bone1Name=dyn_coat_bk_r_05, bone2Name=dyn_coat_bk_r_04, swing1LimitDegrees=20, swing2LimitDegrees=45, twistLimitDegrees=5}
 {bone1Name=dyn_coat_bk_r_06, bone2Name=dyn_coat_bk_r_05, swing1LimitDegrees=10, swing2LimitDegrees=20, twistLimitDegrees=5}
 {bone1Name=dyn_coat_bk_l_05, bone2Name=dyn_coat_bk_l_04, swing1LimitDegrees=10, swing2LimitDegrees=45, twistLimitDegrees=5}
 {bone1Name=dyn_coat_bk_l_06, bone2Name=dyn_coat_bk_l_05, swing1LimitDegrees=10, swing2LimitDegrees=20, twistLimitDegrees=5}
 {bone1Name=dyn_coat_bk_r_04, bone2Name=dyn_coat_bk_r_02, swing1LimitDegrees=20, swing2LimitDegrees=45, twistLimitDegrees=5}
 {bone1Name=dyn_coat_bk_04, bone2Name=dyn_coat_bk_02, swing1LimitDegrees=10, swing2LimitDegrees=15, twistLimitDegrees=5}
 {bone1Name=dyn_coat_bk_l_04, bone2Name=dyn_coat_bk_l_02, swing1LimitDegrees=10, swing2LimitDegrees=45, twistLimitDegrees=5}
 {
  bone1Name=dyn_coat_r_05
  bone2Name=dyn_coat_bk_r_05
  swing1LimitDegrees=45
  swing1Motion=Free
  swing2LimitDegrees=45
  swing2Motion=Free
  twistLimitDegrees=45
  twistMotion=Free
 }
 {
  bone1Name=dyn_coat_l_05
  bone2Name=dyn_coat_bk_l_05
  swing1LimitDegrees=45
  swing1Motion=Free
  swing2LimitDegrees=45
  swing2Motion=Free
  twistLimitDegrees=45
  twistMotion=Free
 }
 {
  bone1Name=dyn_coat_r_04
  bone2Name=dyn_coat_bk_r_02
  swing1LimitDegrees=45
  swing1Motion=Free
  swing2LimitDegrees=45
  swing2Motion=Free
  twistLimitDegrees=45
  twistMotion=Free
 }
 {
  bone1Name=dyn_coat_l_04
  bone2Name=dyn_coat_bk_l_02
  swing1LimitDegrees=45
  swing1Motion=Free
  swing2LimitDegrees=45
  swing2Motion=Free
  twistLimitDegrees=45
  twistMotion=Free
 }
 {bone1Name=dyn_coat_bk_05, bone2Name=dyn_coat_bk_04, swing1LimitDegrees=10, swing2LimitDegrees=15, twistLimitDegrees=5}
 {bone1Name=dyn_coat_bk_06, bone2Name=dyn_coat_bk_05, swing1LimitDegrees=10, swing2LimitDegrees=15, twistLimitDegrees=5}
 {
  bone1Name=calf_r
  bone2Name=thigh_r
  swing1LimitDegrees=45
  swing1Motion=Free
  swing2LimitDegrees=45
  swing2Motion=Free
  twistLimitDegrees=45
  twistMotion=Free
 }
 {
  bone1Name=thigh_r
  bone2Name=pelvis
  swing1LimitDegrees=45
  swing1Motion=Free
  swing2LimitDegrees=45
  swing2Motion=Free
  twistLimitDegrees=45
  twistMotion=Free
 }
 {
  bone1Name=dyn_coat_bk_l_05
  bone2Name=dyn_coat_bk_05
  swing1LimitDegrees=45
  swing1Motion=Free
  swing2LimitDegrees=45
  swing2Motion=Free
  twistLimitDegrees=45
  twistMotion=Free
 }
 {
  bone1Name=dyn_coat_bk_r_05
  bone2Name=dyn_coat_bk_05
  swing1LimitDegrees=45
  swing1Motion=Free
  swing2LimitDegrees=45
  swing2Motion=Free
  twistLimitDegrees=45
  twistMotion=Free
 }
]
