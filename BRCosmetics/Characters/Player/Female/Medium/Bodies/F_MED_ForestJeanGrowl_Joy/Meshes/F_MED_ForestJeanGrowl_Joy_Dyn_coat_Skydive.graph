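# Physics asset for the skydive variant of the dynamic coat on the F_MED_ForestJeanGrowl_Joy female body.
# Capsules on pelvis, spine_01, thighs and calves cover the body, alongside five coat bone chains (front left/right, back left/center/right) with generally larger capsules toward the hem.
# Chains hang from spine_01, the back ones through tiny MSTR anchors, with limited swing and tight twist.
# Angularly free cross-constraints tie neighboring panels together so the coat moves as one.
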PhysicsAsset /BRCosmetics/Characters/Player/Female/Medium/Bodies/F_MED_ForestJeanGrowl_Joy/Meshes/F_MED_ForestJeanGrowl_Joy_Dyn_coat_Skydive
bodies {
 calf_l [{shapeType=Capsule, length=50.6561, radius=16.2157, shapeName=pelvis_capsule, center=<-30.4611 2.7534 5.3559>, rotation=<rot 90 0 0>}]
 calf_r [{shapeType=Capsule, length=50.6561, radius=16.3349, shapeName=pelvis_capsule, center=<30.4611 -2.7534 -5.4696>, rotation=<rot 90 180 0>}]
 dyn_coat_bk_01 [{shapeType=Capsule, length=43.7053, radius=11.5, shapeName=dyn_coat_l_01_capsule, center=<8.7723 0.0007 17>, rotation=<rot 0 0 90>}]
 dyn_coat_bk_02 [{shapeType=Capsule, length=37.9314, radius=13.1507, shapeName=dyn_coat_bk_02_capsule, center=<6.1175 0 10.4772>, rotation=<rot 0 0 90>}]
 dyn_coat_bk_04 [{shapeType=Capsule, length=26.0821, radius=16, shapeName=dyn_coat_bk_02_capsule, center=<0 0 12>, rotation=<rot 0 0 90>}]
 dyn_coat_bk_05 [{shapeType=Capsule, length=40.9348, radius=19, shapeName=dyn_coat_bk_02_capsule, center=<0 0 12>, rotation=<rot 0 0 90>}]
 dyn_coat_bk_06 [{shapeType=Capsule, length=47.9783, radius=19.9395, shapeName=dyn_coat_bk_02_capsule, center=<3.462 0 12>, rotation=<rot 0 0 90>}]
 dyn_coat_bk_MSTR [{shapeType=Capsule, length=0.505, radius=0.505, shapeName=dyn_coat_l_MSTR_capsule}]
 dyn_coat_bk_l_01 [{shapeType=Capsule, length=50.4514, radius=10.4293, shapeName=dyn_coat_l_01_capsule, center=<10.7079 0 17.2013>, rotation=<rot 0 0 90>}]
 dyn_coat_bk_l_02 [{shapeType=Capsule, length=50.7561, radius=15.5, shapeName=dyn_coat_l_01_capsule, center=<10.3356 0 14.3426>, rotation=<rot 0 0 90>}]
 dyn_coat_bk_l_04 [{shapeType=Capsule, length=32.7561, radius=23, shapeName=dyn_coat_l_01_capsule, center=<10.3356 0 13.044>, rotation=<rot 0 0 90>}]
 dyn_coat_bk_l_05 [{shapeType=Capsule, length=43.3952, radius=24.2858, shapeName=dyn_coat_bk_r_05_capsule, center=<1.8615 0 15>, rotation=<rot 0 0 90>}]
 dyn_coat_bk_l_06 [{shapeType=Capsule, length=54.0134, radius=19.9395, shapeName=dyn_coat_bk_02_capsule, center=<5.881 -0.8946 5>, rotation=<rot 0 0 90>}]
 dyn_coat_bk_l_MSTR [{shapeType=Capsule, length=0.505, radius=0.505, shapeName=dyn_coat_l_MSTR_capsule}]
 dyn_coat_bk_r_01 [{shapeType=Capsule, length=50.3014, radius=11.5, shapeName=dyn_coat_l_01_capsule, center=<10.7079 0 17.2013>, rotation=<rot 0 0 90>}]
 dyn_coat_bk_r_02 [{shapeType=Capsule, length=50.7561, radius=15.5, shapeName=dyn_coat_l_01_capsule, center=<10.3356 0 14.3426>, rotation=<rot 0 0 90>}]
 dyn_coat_bk_r_04 [{shapeType=Capsule, length=32.7561, radius=23, shapeName=dyn_coat_l_01_capsule, center=<10.3356 0 13.044>, rotation=<rot 0 0 90>}]
 dyn_coat_bk_r_05 [{shapeType=Capsule, length=43.3952, radius=24.1665, shapeName=dyn_coat_bk_r_05_capsule, center=<1.4783 0 15>, rotation=<rot 0 0 90>}]
 dyn_coat_bk_r_06 [{
   shapeType=Capsule
   length=47.9783
   radius=19.9395
   shapeName=dyn_coat_bk_02_capsule
   center=<7.3484 1.0413 12>
   rotation=<rot -0.0001 -0.0001 80>
  }]
 dyn_coat_bk_r_MSTR [{shapeType=Capsule, length=0.505, radius=0.505, shapeName=dyn_coat_l_MSTR_capsule}]
 dyn_coat_l_02 [{shapeType=Capsule, length=12.2561, radius=8, shapeName=dyn_coat_l_01_capsule, center=<5.44 0 5.5818>, rotation=<rot 90 0 0>}]
 dyn_coat_l_03 [{shapeType=Capsule, length=20.2561, radius=9, shapeName=dyn_coat_l_01_capsule, center=<10.3356 0 12.3755>, rotation=<rot 0 0 90>}]
 dyn_coat_l_04 [{shapeType=Capsule, length=19.0557, radius=16.7351, shapeName=dyn_coat_l_01_capsule, center=<10.3356 0 14.8748>, rotation=<rot 0 0 90>}]
 dyn_coat_l_05 [{
   shapeType=Capsule
   length=25.1064
   radius=19.8997
   shapeName=dyn_coat_l_01_capsule
   center=<3.5998 -5 14>
   rotation=<rot -0.0003 0.0008 109.9999>
  }]
 dyn_coat_l_06 [{
   shapeType=Capsule
   length=48.785
   radius=18.7463
   shapeName=dyn_coat_bk_02_capsule
   center=<5.1676 -2.4355 15.2199>
   rotation=<rot 9.8464 -17.4949 98.4813>
  }]
 dyn_coat_r_02 [{shapeType=Capsule, length=12.2561, radius=8, shapeName=dyn_coat_l_01_capsule, center=<5.44 0 5.5818>, rotation=<rot 90 0 0>}]
 dyn_coat_r_03 [{shapeType=Capsule, length=20.2561, radius=9, shapeName=dyn_coat_l_01_capsule, center=<10.3356 0 12.3755>, rotation=<rot 0 0 90>}]
 dyn_coat_r_04 [{shapeType=Capsule, length=27.7561, radius=17.5, shapeName=dyn_coat_l_01_capsule, center=<10.3356 0 15.4045>, rotation=<rot 0 0 90>}]
 dyn_coat_r_05 [{
   shapeType=Capsule
   length=26.1277
   radius=22.6193
   shapeName=dyn_coat_l_01_capsule
   center=<3.5998 0 16.4623>
   rotation=<rot -0.0003 -0.0008 70.0001>
  }]
 dyn_coat_r_06 [{
   shapeType=Capsule
   length=52.3174
   radius=20.3956
   shapeName=dyn_coat_bk_02_capsule
   center=<3.462 0 12>
   rotation=<rot 7.6442 6.4661 70.4325>
  }]
 pelvis [{shapeType=Capsule, length=11.5, radius=15.021, shapeName=pelvis_capsule, center=<-4.8171 -0.6518 -0.0005>, rotation=<rot 90 -0.0002 0>}]
 spine_01 [{shapeType=Capsule, length=11.4638, radius=12, shapeName=pelvis_capsule, center=<-0.9991 -0.3721 0>, rotation=<rot -90 -0.0005 0>}]
 thigh_l [{
   shapeType=Capsule
   length=31.089
   radius=13.0541
   shapeName=pelvis_capsule
   center=<-20.1408 2.1742 -0.7756>
   rotation=<rot 88.6498 0.0001 -179.9998>
  }]
 thigh_r [{
   shapeType=Capsule
   length=31.089
   radius=13.0541
   shapeName=pelvis_capsule
   center=<20.0636 2.1741 -1.7232>
   rotation=<rot 88.6498 0.0001 -179.9998>
  }]
}
constraints [
 {bone1Name=dyn_coat_l_02, bone2Name=spine_01, swing1LimitDegrees=25, swing2LimitDegrees=45, twistLimitDegrees=5}
 {bone1Name=dyn_coat_l_03, bone2Name=dyn_coat_l_02, swing1LimitDegrees=15, swing2LimitDegrees=45, twistLimitDegrees=5}
 {bone1Name=dyn_coat_l_04, bone2Name=dyn_coat_l_03, swing1LimitDegrees=15, swing2LimitDegrees=45, twistLimitDegrees=5}
 {bone1Name=dyn_coat_l_05, bone2Name=dyn_coat_l_04, swing1LimitDegrees=10, swing2LimitDegrees=45, twistLimitDegrees=5}
 {bone1Name=dyn_coat_l_06, bone2Name=dyn_coat_l_05, swing1LimitDegrees=10, swing2LimitDegrees=45, twistLimitDegrees=5}
 {bone1Name=dyn_coat_r_02, bone2Name=spine_01, swing1LimitDegrees=25, swing2LimitDegrees=45, twistLimitDegrees=5}
 {bone1Name=dyn_coat_r_03, bone2Name=dyn_coat_r_02, swing1LimitDegrees=10, swing2LimitDegrees=45, twistLimitDegrees=5}
 {bone1Name=dyn_coat_r_04, bone2Name=dyn_coat_r_03, swing1LimitDegrees=10, swing2LimitDegrees=45, twistLimitDegrees=5}
 {bone1Name=dyn_coat_r_05, bone2Name=dyn_coat_r_04, swing1LimitDegrees=10, swing2LimitDegrees=45, twistLimitDegrees=5}
 {bone1Name=dyn_coat_r_06, bone2Name=dyn_coat_r_05, swing1LimitDegrees=10, swing2LimitDegrees=45, twistLimitDegrees=5}
 {bone1Name=dyn_coat_bk_01, bone2Name=dyn_coat_bk_MSTR, swing1LimitDegrees=10, swing2LimitDegrees=15, twistLimitDegrees=5}
 {bone1Name=dyn_coat_bk_02, bone2Name=dyn_coat_bk_01, swing1LimitDegrees=10, swing2LimitDegrees=15, twistLimitDegrees=5}
 {bone1Name=dyn_coat_bk_l_01, bone2Name=dyn_coat_bk_l_MSTR, swing1LimitDegrees=10, swing2LimitDegrees=45, twistLimitDegrees=5}
 {bone1Name=dyn_coat_bk_l_02, bone2Name=dyn_coat_bk_l_01, swing1LimitDegrees=10, swing2LimitDegrees=45, twistLimitDegrees=5}
 {bone1Name=dyn_coat_bk_r_01, bone2Name=dyn_coat_bk_r_MSTR, swing1LimitDegrees=10, swing2LimitDegrees=45, twistLimitDegrees=5}
 {bone1Name=dyn_coat_bk_r_02, bone2Name=dyn_coat_bk_r_01, swing1LimitDegrees=15, swing2LimitDegrees=45, twistLimitDegrees=5}
 {
  bone1Name=calf_l
  bone2Name=thigh_l
  swing1LimitDegrees=45
  swing1Motion=Free
  swing2LimitDegrees=45
  swing2Motion=Free
  twistLimitDegrees=45
  twistMotion=Free
 }
 {
  bone1Name=dyn_coat_bk_MSTR
  bone2Name=spine_01
  swing1LimitDegrees=45
  swing1Motion=Free
  swing2LimitDegrees=45
  swing2Motion=Free
  twistLimitDegrees=45
  twistMotion=Free
 }
 {
  bone1Name=dyn_coat_bk_l_MSTR
  bone2Name=spine_01
  swing1LimitDegrees=45
  swing1Motion=Free
  swing2LimitDegrees=45
  swing2Motion=Free
  twistLimitDegrees=45
  twistMotion=Free
 }
 {
  bone1Name=dyn_coat_bk_r_MSTR
  bone2Name=spine_01
  swing1LimitDegrees=45
  swing1Motion=Free
  swing2LimitDegrees=45
  swing2Motion=Free
  twistLimitDegrees=45
  twistMotion=Free
 }
 {
  bone1Name=thigh_l
  bone2Name=pelvis
  swing1LimitDegrees=45
  swing1Motion=Free
  swing2LimitDegrees=45
  swing2Motion=Free
  twistLimitDegrees=45
  twistMotion=Free
 }
 {
  bone1Name=spine_01
  bone2Name=pelvis
  swing1LimitDegrees=45
  swing1Motion=Free
  swing2LimitDegrees=45
  swing2Motion=Free
  twistLimitDegrees=45
  twistMotion=Free
 }
 {bone1Name=dyn_coat_bk_r_05, bone2Name=dyn_coat_bk_r_04, swing1LimitDegrees=20, swing2LimitDegrees=45, twistLimitDegrees=5}
 {bone1Name=dyn_coat_bk_r_06, bone2Name=dyn_coat_bk_r_05, swing1LimitDegrees=10, swing2LimitDegrees=20, twistLimitDegrees=5}
 {bone1Name=dyn_coat_bk_l_05, bone2Name=dyn_coat_bk_l_04, swing1LimitDegrees=10, swing2LimitDegrees=45, twistLimitDegrees=5}
 {bone1Name=dyn_coat_bk_l_06, bone2Name=dyn_coat_bk_l_05, swing1LimitDegrees=10, swing2LimitDegrees=20, twistLimitDegrees=5}
 {bone1Name=dyn_coat_bk_r_04, bone2Name=dyn_coat_bk_r_02, swing1LimitDegrees=20, swing2LimitDegrees=45, twistLimitDegrees=5}
 {bone1Name=dyn_coat_bk_04, bone2Name=dyn_coat_bk_02, swing1LimitDegrees=10, swing2LimitDegrees=15, twistLimitDegrees=5}
 {bone1Name=dyn_coat_bk_l_04, bone2Name=dyn_coat_bk_l_02, swing1LimitDegrees=10, swing2LimitDegrees=45, twistLimitDegrees=5}
 {
  bone1Name=dyn_coat_r_05
  bone2Name=dyn_coat_bk_r_05
  swing1LimitDegrees=45
  swing1Motion=Free
  swing2LimitDegrees=45
  swing2Motion=Free
  twistLimitDegrees=45
  twistMotion=Free
 }
 {
  bone1Name=dyn_coat_l_05
  bone2Name=dyn_coat_bk_l_05
  swing1LimitDegrees=45
  swing1Motion=Free
  swing2LimitDegrees=45
  swing2Motion=Free
  twistLimitDegrees=45
  twistMotion=Free
 }
 {
  bone1Name=dyn_coat_r_04
  bone2Name=dyn_coat_bk_r_02
  swing1LimitDegrees=45
  swing1Motion=Free
  swing2LimitDegrees=45
  swing2Motion=Free
  twistLimitDegrees=45
  twistMotion=Free
 }
 {
  bone1Name=dyn_coat_l_04
  bone2Name=dyn_coat_bk_l_02
  swing1LimitDegrees=45
  swing1Motion=Free
  swing2LimitDegrees=45
  swing2Motion=Free
  twistLimitDegrees=45
  twistMotion=Free
 }
 {bone1Name=dyn_coat_bk_05, bone2Name=dyn_coat_bk_04, swing1LimitDegrees=10, swing2LimitDegrees=15, twistLimitDegrees=5}
 {bone1Name=dyn_coat_bk_06, bone2Name=dyn_coat_bk_05, swing1LimitDegrees=10, swing2LimitDegrees=15, twistLimitDegrees=5}
 {
  bone1Name=calf_r
  bone2Name=thigh_r
  swing1LimitDegrees=45
  swing1Motion=Free
  swing2LimitDegrees=45
  swing2Motion=Free
  twistLimitDegrees=45
  twistMotion=Free
 }
 {
  bone1Name=thigh_r
  bone2Name=pelvis
  swing1LimitDegrees=45
  swing1Motion=Free
  swing2LimitDegrees=45
  swing2Motion=Free
  twistLimitDegrees=45
  twistMotion=Free
 }
 {
  bone1Name=dyn_coat_bk_l_05
  bone2Name=dyn_coat_bk_05
  swing1LimitDegrees=45
  swing1Motion=Free
  swing2LimitDegrees=45
  swing2Motion=Free
  twistLimitDegrees=45
  twistMotion=Free
 }
 {
  bone1Name=dyn_coat_bk_r_05
  bone2Name=dyn_coat_bk_05
  swing1LimitDegrees=45
  swing1Motion=Free
  swing2LimitDegrees=45
  swing2Motion=Free
  twistLimitDegrees=45
  twistMotion=Free
 }
]
