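# Physics asset for the sleeves of the female medium ForestJeanGrowl_Joy cosmetic body, built entirely from capsule bodies.
# Top, bottom and back sleeve-bone chains hang from each forearm with limited swing and twist, the back chains through a freely rotating MSTR body.
# Pelvis, thigh, forearm and hand capsules cover the surrounding body, linked by free angular constraints.
# The sides are asymmetric: the right bottom chain has an extra bone, and several limits differ.
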PhysicsAsset /BRCosmetics/Characters/Player/Female/Medium/Bodies/F_MED_ForestJeanGrowl_Joy/Meshes/F_MED_ForestJeanGrowl_Joy_Sleeve_Physics
bodies {
 dyn_sleeve_l_bk_01 [{shapeType=Capsule, length=6, radius=3.5, shapeName=dyn_sleeve_l_bk_01_capsule, center=<5 -0.5 0>, rotation=<rot 90 0 0>}]
 dyn_sleeve_l_bk_02 [{shapeType=Capsule, length=6, radius=3.5, shapeName=dyn_sleeve_l_bk_02_capsule, center=<5 -0.5 0>, rotation=<rot 90 0 0>}]
 dyn_sleeve_l_bk_MSTR [{shapeType=Capsule, length=0, radius=1, shapeName=dyn_sleeve_l_bk_MSTR_capsule}]
 dyn_sleeve_l_bot_02 [{shapeType=Capsule, length=6, radius=3.5, shapeName=dyn_sleeve_l_bot_02_capsule, center=<4.3097 0 0>, rotation=<rot 90 0 0>}]
 dyn_sleeve_l_bot_03 [{shapeType=Capsule, length=9.5, radius=3.5, shapeName=dyn_sleeve_l_bot_03_capsule, center=<9.3006 0 0>, rotation=<rot 90 0 0>}]
 dyn_sleeve_l_top_02 [{shapeType=Capsule, length=9, radius=5.5, shapeName=dyn_sleeve_r_bot_02_capsule, center=<4.3 -2 1.75>, rotation=<rot -90 0 0>}]
 dyn_sleeve_l_top_03 [{shapeType=Capsule, length=10.5, radius=4, shapeName=dyn_sleeve_r_bot_02_capsule, center=<9 -1.5 0.5>, rotation=<rot -90 -0.0005 0>}]
 dyn_sleeve_r_bk_01 [{shapeType=Capsule, length=6, radius=3.5, shapeName=dyn_sleeve_r_bk_01_capsule, center=<-3 1.5 0>, rotation=<rot -90 0 0>}]
 dyn_sleeve_r_bk_02 [{shapeType=Capsule, length=6, radius=3.5, shapeName=dyn_sleeve_r_bk_02_capsule, center=<-3 1.5 0>, rotation=<rot -90 0 0>}]
 dyn_sleeve_r_bk_MSTR [{shapeType=Capsule, length=0, radius=1, shapeName=dyn_sleeve_r_bk_MSTR_capsule}]
 dyn_sleeve_r_bot_01 [{shapeType=Capsule, length=6, radius=3.5, shapeName=dyn_sleeve_r_bot_01_capsule, center=<-3 0 0>, rotation=<rot -90 0 0>}]
 dyn_sleeve_r_bot_02 [{shapeType=Capsule, length=6, radius=3.5, shapeName=dyn_sleeve_r_bot_02_capsule, center=<-3.5776 0 0>, rotation=<rot -90 0 0>}]
 dyn_sleeve_r_bot_03 [{shapeType=Capsule, length=10.1894, radius=3.8106, shapeName=dyn_sleeve_r_bot_03_capsule, center=<-7 0 0>, rotation=<rot -90 0.0001 0>}]
 dyn_sleeve_r_top_02 [{shapeType=Capsule, length=9, radius=5.5, shapeName=dyn_sleeve_r_bot_02_capsule, center=<-4.3 2 1.75>, rotation=<rot -90 0 0>}]
 dyn_sleeve_r_top_03 [{shapeType=Capsule, length=10.5, radius=4, shapeName=dyn_sleeve_r_bot_02_capsule, center=<-9 1.5 0.5>, rotation=<rot -90 -0.0005 0>}]
 hand_l [{shapeType=Capsule, length=7, radius=5.5, shapeName=hand_l_capsule, center=<5.7 -1.5 0>, rotation=<rot 10 -90 -85.9729>}]
 hand_r [{shapeType=Capsule, length=7, radius=5.5, shapeName=hand_r_capsule, center=<-5.7 1.5 0>, rotation=<rot 10 -90 -91.2856>}]
 lowerarm_l [{shapeType=Capsule, length=22, radius=4, shapeName=lowerarm_l_capsule, center=<14 0 0>, rotation=<rot 85 -5.0003 -0.0003>}]
 lowerarm_r [{shapeType=Capsule, length=22, radius=4, shapeName=lowerarm_r_capsule, center=<-14 0 -0.1169>, rotation=<rot 90 -4.9999 0>}]
 pelvis [{shapeType=Capsule, length=17.2656, radius=11.0684, shapeName=pelvis_capsule, center=<4.1726 -1.7778 0>, rotation=<rot 90 -0.0005 0>}]
 thigh_l [{
   shapeType=Capsule
   length=39.3088
   radius=10.9623
   shapeName=thigh_r_capsule
   center=<-18.4013 -3.8649 0.2461>
   rotation=<rot 85 179.9995 179.9998>
  }]
 thigh_r [{
   shapeType=Capsule
   length=39.3088
   radius=10.9623
   shapeName=thigh_r_capsule
   center=<18.4013 3.8649 0.2461>
   rotation=<rot 85 179.9995 179.9998>
  }]
}
constraints [
 {
  bone1Name=hand_l
  bone2Name=lowerarm_l
  swing1LimitDegrees=45
  swing1Motion=Free
  swing2LimitDegrees=45
  swing2Motion=Free
  twistLimitDegrees=45
  twistMotion=Free
 }
 {
  bone1Name=lowerarm_l
  bone2Name=pelvis
  swing1LimitDegrees=45
  swing1Motion=Free
  swing2LimitDegrees=45
  swing2Motion=Free
  twistLimitDegrees=45
  twistMotion=Free
 }
 {bone1Name=dyn_sleeve_l_top_03, bone2Name=dyn_sleeve_l_top_02, swing1LimitDegrees=60, swing2LimitDegrees=60, twistLimitDegrees=60}
 {bone1Name=dyn_sleeve_l_top_02, bone2Name=lowerarm_l, swing1LimitDegrees=10, swing2LimitDegrees=10, twistLimitDegrees=10}
 {bone1Name=dyn_sleeve_l_bot_03, bone2Name=dyn_sleeve_l_bot_02, swing1LimitDegrees=60, swing2LimitDegrees=60, twistLimitDegrees=60}
 {bone1Name=dyn_sleeve_l_bot_02, bone2Name=lowerarm_l, swing1LimitDegrees=10, swing2LimitDegrees=10, twistLimitDegrees=10}
 {bone1Name=dyn_sleeve_l_bk_02, bone2Name=dyn_sleeve_l_bk_01, swing1LimitDegrees=20, swing2LimitDegrees=20, twistLimitDegrees=2}
 {bone1Name=dyn_sleeve_l_bk_01, bone2Name=dyn_sleeve_l_bk_MSTR, swing1LimitDegrees=20, swing2LimitDegrees=20, twistLimitDegrees=2}
 {
  bone1Name=dyn_sleeve_l_bk_MSTR
  bone2Name=lowerarm_l
  swing1LimitDegrees=45
  swing1Motion=Free
  swing2LimitDegrees=45
  swing2Motion=Free
  twistLimitDegrees=45
  twistMotion=Free
 }
 {
  bone1Name=lowerarm_r
  bone2Name=pelvis
  swing1LimitDegrees=45
  swing1Motion=Free
  swing2LimitDegrees=45
  swing2Motion=Free
  twistLimitDegrees=45
  twistMotion=Free
 }
 {
  bone1Name=dyn_sleeve_r_bk_MSTR
  bone2Name=lowerarm_r
  swing1LimitDegrees=45
  swing1Motion=Free
  swing2LimitDegrees=45
  swing2Motion=Free
  twistLimitDegrees=45
  twistMotion=Free
 }
 {bone1Name=dyn_sleeve_r_bk_01, bone2Name=dyn_sleeve_r_bk_MSTR, swing1LimitDegrees=45, swing2LimitDegrees=45, twistLimitDegrees=45}
 {bone1Name=dyn_sleeve_r_bk_02, bone2Name=dyn_sleeve_r_bk_01, swing1LimitDegrees=45, swing2LimitDegrees=45, twistLimitDegrees=45}
 {bone1Name=dyn_sleeve_r_bot_01, bone2Name=lowerarm_r, swing1LimitDegrees=60, swing2LimitDegrees=60, twistLimitDegrees=60}
 {bone1Name=dyn_sleeve_r_bot_02, bone2Name=dyn_sleeve_r_bot_01, swing1LimitDegrees=10, swing2LimitDegrees=10, twistLimitDegrees=10}
 {bone1Name=dyn_sleeve_r_bot_03, bone2Name=dyn_sleeve_r_bot_02, swing1LimitDegrees=60, swing2LimitDegrees=60, twistLimitDegrees=60}
 {
  bone1Name=hand_r
  bone2Name=lowerarm_r
  swing1LimitDegrees=45
  swing1Motion=Free
  swing2LimitDegrees=45
  swing2Motion=Free
  twistLimitDegrees=45
  twistMotion=Free
 }
 {
  bone1Name=thigh_l
  bone2Name=pelvis
  swing1LimitDegrees=45
  swing1Motion=Free
  swing2LimitDegrees=45
  swing2Motion=Free
  twistLimitDegrees=45
  twistMotion=Free
 }
 {
  bone1Name=thigh_r
  bone2Name=pelvis
  swing1LimitDegrees=45
  swing1Motion=Free
  swing2LimitDegrees=45
  swing2Motion=Free
  twistLimitDegrees=45
  twistMotion=Free
 }
 {bone1Name=dyn_sleeve_r_top_02, bone2Name=lowerarm_r, swing1LimitDegrees=10, swing2LimitDegrees=10, twistLimitDegrees=10}
 {bone1Name=dyn_sleeve_r_top_03, bone2Name=dyn_sleeve_r_top_02, swing1LimitDegrees=60, swing2LimitDegrees=60, twistLimitDegrees=35}
]
