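# Animation Blueprint for the F_MED_ForestJeanGrowl_Joy character body, adding secondary motion to its coat and sleeves over the pose copied from the parent mesh.
# RigidBody simulation drives sleeves, and the coat when cloth is off; without rigid bodies, Trail and AnimDynamics chains or a posed DBNO variant take over.
# Sleeve constraints, skydive and lobby offsets, and a Control Rig finish the pose.
# The event graph updates gravity, cloth, wind and movement-state inputs.
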AnimBlueprint /BRCosmetics/Characters/Player/Female/Medium/Bodies/F_MED_ForestJeanGrowl_Joy/Meshes/F_MED_ForestJeanGrowl_Joy_AnimBP
target_skeleton=&/F_MED_ForestJeanGrowl_Joy_Skeleton
graphs [
 AnimGraph {nodes [
   n1 = Blend Poses by bool(BlendPose_0=Use cached pose 'dbnoCache'(), BlendPose_1=Use cached pose 'trailCache'(), bActiveValue=Get bIsDBNO()) {blendType=Linear}
   n2 = Use cached pose 'RbanCache'()
   n3 = Get bRigidBodyEnabled()
   n4 = "Constraint - dyn_sleeve_l_bot_01 (R)"(
    ComponentPose=Local To Component(LocalPose=Blend Poses by bool(BlendPose_0=@n2, BlendPose_1=@n1, bActiveValue=@n3) {blendType=Linear})
    Alpha=1
    ConstraintWeights_0=1
   ) {
    alphaBoolBlend {blendOption=Linear}
    alphaScaleBiasClamp {bClampResult=true, clampMax=0.15}
    boneToModify {boneName=dyn_sleeve_l_bot_01}
    constraintSetup [{transformType=Rotation, targetBone {boneName=dyn_sleeve_l_bk_01}}]
    constraintWeights [1]
   }
   n5 = "Constraint - dyn_sleeve_l_bot_02 (R)"(ComponentPose=@n4, Alpha=1, ConstraintWeights_0=1) {
    alphaBoolBlend {blendOption=Linear}
    alphaScaleBiasClamp {bClampResult=true, clampMax=0.15}
    boneToModify {boneName=dyn_sleeve_l_bot_02}
    constraintSetup [{transformType=Rotation, targetBone {boneName=dyn_sleeve_l_bk_02}}]
    constraintWeights [1]
   }
   n6 = "Constraint - dyn_sleeve_r_bot_01 (R)"(ComponentPose=@n5, Alpha=1, ConstraintWeights_0=1) {
    alphaBoolBlend {blendOption=Linear}
    alphaScaleBiasClamp {bClampResult=true, clampMax=0.15}
    boneToModify {boneName=dyn_sleeve_r_bot_01}
    constraintSetup [{transformType=Rotation, targetBone {boneName=dyn_sleeve_r_bk_01}}]
    constraintWeights [1]
   }
   n7 = "Constraint - dyn_sleeve_r_bot_02 (R)"(ComponentPose=@n6, Alpha=1, ConstraintWeights_0=1) {
    alphaBoolBlend {blendOption=Linear}
    alphaScaleBiasClamp {bClampResult=true, clampMax=0.15}
    boneToModify {boneName=dyn_sleeve_r_bot_02}
    constraintSetup [{transformType=Rotation, targetBone {boneName=dyn_sleeve_r_bk_02}}]
    constraintWeights [1]
   }
   n8 = "Constraint - dyn_sleeve_l_top_02 (RR)"!(ComponentPose=@n7, Alpha=1, ConstraintWeights_0=0.7, ConstraintWeights_1:float=0.3) -> (Pose) {
    alphaBoolBlend {blendOption=Linear}
    alphaScaleBiasClamp {bClampResult=true, clampMax=0.15}
    boneToModify {boneName=dyn_sleeve_l_top_02}
    constraintSetup [{transformType=Rotation, targetBone {boneName=dyn_sleeve_l_bk_02}}, {transformType=Rotation, targetBone {boneName=lowerarm_l}}]
    constraintWeights [1,1]
   }
   n9 = "Constraint - dyn_sleeve_r_top_02 (RR)"!(ComponentPose=@n8, Alpha=1, ConstraintWeights_0=0.7, ConstraintWeights_1:float=0.3) -> (Pose) {
    alphaBoolBlend {blendOption=Linear}
    alphaScaleBiasClamp {bClampResult=true, clampMax=0.15}
    boneToModify {boneName=dyn_sleeve_r_top_02}
    constraintSetup [{transformType=Rotation, targetBone {boneName=dyn_sleeve_r_bk_02}}, {transformType=Rotation, targetBone {boneName=lowerarm_r}}]
    constraintWeights [1,1]
   }
   n10 = "Constraint - dyn_sleeve_l_bk_01 (PP)"!(ComponentPose=@n9, Alpha=1, ConstraintWeights_0=0.8, ConstraintWeights_1:float=0.2) -> (Pose) {
    alphaBoolBlend {blendOption=Linear}
    alphaScaleBiasClamp {bClampResult=true, clampMax=0.3}
    boneToModify {boneName=dyn_sleeve_l_bk_01}
    constraintSetup [{targetBone {boneName=deform_elbow_out_l}}, {targetBone {boneName=lowerarm_l}}]
    constraintWeights [1,1]
   }
   n11 = "Constraint - dyn_sleeve_r_bk_01 (PP)"!(ComponentPose=@n10, Alpha=1, ConstraintWeights_0=0.8, ConstraintWeights_1:float=0.2) -> (Pose) {
    alphaBoolBlend {blendOption=Linear}
    alphaScaleBiasClamp {bClampResult=true, clampMax=0.3}
    boneToModify {boneName=dyn_sleeve_r_bk_01}
    constraintSetup [{targetBone {boneName=deform_elbow_out_r}}, {targetBone {boneName=lowerarm_r}}]
    constraintWeights [1,1]
   }
   n12 = Get Skydive() -> (Skydive:double)
   n13 = "Transform (Modify) Bone - Bone: dyn_sleeve_r_top_MSTR"(ComponentPose=@n11, Alpha=@n12, Translation="0,0,0", Rotation="0,0,-5", Scale="1,1,1") {rotationMode=BMM_Additive, translationMode=BMM_Additive, boneToModify {boneName=dyn_sleeve_r_top_MSTR}, rotation {roll=-5}}
   n14 = "Transform (Modify) Bone - Bone: dyn_sleeve_l_top_MSTR"(ComponentPose=@n13, Alpha=@n12, Translation="0,0,0", Rotation="0,0,-10", Scale="1,1,1") {rotationMode=BMM_Additive, translationMode=BMM_Additive, boneToModify {boneName=dyn_sleeve_l_top_MSTR}, rotation {roll=-10}}
   n15 = Get FrontEnd Idle Alpha() -> (FrontEnd Idle Alpha:double)
   n16 = "Transform (Modify) Bone - Bone: dyn_sleeve_r_top_MSTR"(ComponentPose=@n14, Alpha=@n15, Translation="0,0,0", Rotation="0,0,-5", Scale="1,1,1") {rotationMode=BMM_Additive, translationMode=BMM_Additive, boneToModify {boneName=dyn_sleeve_r_top_MSTR}, rotation {roll=-5}}
   n17 = "Transform (Modify) Bone - Bone: dyn_sleeve_l_top_MSTR"(ComponentPose=@n16, Alpha=@n15, Translation="0,0,0", Rotation="5.000,-0.000,5.000", Scale="1,1,1") {rotationMode=BMM_Additive, translationMode=BMM_Additive, boneToModify {boneName=dyn_sleeve_l_top_MSTR}, rotation {pitch=5, roll=5}}
   n18 = "Transform (Modify) Bone - Bone: dyn_sleeve_r_bot_MSTR"(ComponentPose=@n17, Alpha=@n15, Translation="0,0,0", Rotation="-10.297,1.737,4.619", Scale="1,1,1") {rotationMode=BMM_Additive, translationMode=BMM_Additive, boneToModify {boneName=dyn_sleeve_r_bot_MSTR}, rotation=<rot -10.297 1.737 4.619>}
   n19 = "Transform (Modify) Bone - Bone: dyn_sleeve_l_bot_MSTR"(ComponentPose=@n18, Alpha=@n15, Translation="0,0,0", Rotation="4.981,-0.437,-5.0019", Scale="1,1,1") {rotationMode=BMM_Additive, translationMode=BMM_Additive, boneToModify {boneName=dyn_sleeve_l_bot_MSTR}, rotation=<rot 4.981 -0.437 -5.0019>}
   n20 = "Transform (Modify) Bone - Bone: dyn_sleeve_r_bot_03"(ComponentPose=@n19, Alpha=@n15, Translation="0,0,0", Rotation="-10.297,1.737,4.619", Scale="1,1,1") {rotationMode=BMM_Additive, translationMode=BMM_Additive, boneToModify {boneName=dyn_sleeve_r_bot_03}, rotation=<rot -10.297 1.737 4.619>}
   n21 = "Transform (Modify) Bone - Bone: dyn_sleeve_l_bot_03"(ComponentPose=@n20, Alpha=@n15, Translation="0,0,0", Rotation="10.280,-4.281,4.639", Scale="1,1,1") {rotationMode=BMM_Additive, translationMode=BMM_Additive, boneToModify {boneName=dyn_sleeve_l_bot_03}, rotation=<rot 10.28 -4.281 4.639>}
   n22 = Control Rig!(Source=Component To Local(ComponentPose=@n21), Alpha, RigidBody:bool=Get bRigidBodyEnabled(), Emote:double=Get Emote() -> (Emote:double)) -> (Pose) {
    alphaBoolBlend {blendOption=Linear}
    controlRigAssetReference {blueprintRigClass=&/F_MED_ForestJeanGrowl_Joy_CtrlRig.F_MED_ForestJeanGrowl_Joy_CtrlRig_C}
   }
   Output Pose(Result=@n22) {layerGroup=None}
   n23 = "Transform (Modify) Bone - Bone: dyn_ground"!(
    ComponentPose=Local To Component(LocalPose=Copy Pose From Mesh() {bUseAttachedParent=true})
    bAlphaBoolEnabled:bool=Get bIsOnGrindRail()
    Translation="0,0,-100"
    Rotation="0,0,0"
    Scale="1,1,1"
   ) -> (Pose) {
    alphaInputType=Bool
    translationMode=BMM_Additive
    translationSpace=BCS_BoneSpace
    alphaBoolBlend {blendOption=Linear}
    boneToModify {boneName=dyn_ground}
    translation {z=-100}
   }
   n24 = "Transform (Modify) Bone - Bone: attach_backpack"!(
    ComponentPose=@n23
    bAlphaBoolEnabled:bool=Get ForestJeanGrowl_Backpack() -> (ForestJeanGrowl_Backpack:bool)
    Translation="0,0,7"
    Rotation="-5,0,0"
    Scale="1,1,0.9"
   ) -> (Pose) {
    alphaInputType=Bool
    rotationMode=BMM_Additive
    scaleMode=BMM_Additive
    scaleSpace=BCS_BoneSpace
    translationMode=BMM_Additive
    translationSpace=BCS_BoneSpace
    alphaBoolBlend {blendOption=Linear}
    boneToModify {boneName=attach_backpack}
    rotation {pitch=-5}
    scale {z=0.9}
    translation {z=7}
   }
   Save cached pose 'StartupCache'(Pose=Component To Local(ComponentPose=@n24)) {cacheName=StartupCache}
   n25 = Get BwdMovement() -> (BwdMovement:double)
   n26 = Trail controller - Bone: dyn_coat_l_06(ComponentPose=Local To Component(LocalPose=Use cached pose 'StartupCache'()), Alpha=@n25) {
    bUsePlanarLimit=true
    chainLength=4
    debugLifeTime=0.1
    maxDeltaTime=0.333
    alphaScaleBiasClamp {bInterpResult=true}
    rotationLimits [{limitMax=<180 180 180>}, {limitMax=<180 180 180>}, {limitMax=<180 180 180>}, {limitMax=<180 180 180>}]
    rotationOffsets [<0 0 0>, <0 0 0>, <0 0 0>, <0 0 0>]
    trailBone {boneName=dyn_coat_l_06}
    trailRelaxationSpeed {editorCurveData {keys [<at -0.0134 19.3841 RCIM_Linear>,<at 0.5079 12.8986 RCIM_Linear>,<at 0.9955 9.6044 RCIM_Linear>]}}
   }
   n27 = Trail controller - Bone: dyn_coat_bk_l_06(ComponentPose=@n26, Alpha=@n25) {
    bUsePlanarLimit=true
    chainLength=4
    debugLifeTime=0.1
    maxDeltaTime=0.333
    alphaScaleBiasClamp {bInterpResult=true}
    rotationLimits [{limitMax=<180 180 180>}, {limitMax=<180 180 180>}, {limitMax=<180 180 180>}, {limitMax=<180 180 180>}]
    rotationOffsets [<0 0 0>, <0 0 0>, <0 0 0>, <0 0 0>]
    trailBone {boneName=dyn_coat_bk_l_06}
    trailRelaxationSpeed {editorCurveData {keys [<at -0.0134 19.3841 RCIM_Linear>,<at 0.5079 12.8986 RCIM_Linear>,<at 0.9955 9.6044 RCIM_Linear>]}}
   }
   n28 = Trail controller - Bone: dyn_coat_bk_04(ComponentPose=@n27, Alpha=@n25) {
    bUsePlanarLimit=true
    debugLifeTime=0.1
    maxDeltaTime=0.0333
    alphaScaleBiasClamp {bInterpResult=true}
    rotationLimits [{limitMax=<180 180 180>}, {limitMax=<180 180 180>}]
    rotationOffsets [<0 0 0>, <0 0 0>]
    trailBone {boneName=dyn_coat_bk_04}
    trailRelaxationSpeed {editorCurveData {keys [<at 0 15 RCIM_Linear>,<at 0.5289 13.1838 RCIM_Linear>,<at 1 6 RCIM_Linear>]}}
   }
   n29 = Trail controller - Bone: dyn_coat_bk_r_06(ComponentPose=@n28, Alpha=@n25) {
    bUsePlanarLimit=true
    chainLength=4
    debugLifeTime=0.1
    maxDeltaTime=0.333
    alphaScaleBiasClamp {bInterpResult=true}
    rotationLimits [{limitMax=<180 180 180>}, {limitMax=<180 180 180>}, {limitMax=<180 180 180>}, {limitMax=<180 180 180>}]
    rotationOffsets [<0 0 0>, <0 0 0>, <0 0 0>, <0 0 0>]
    trailBone {boneName=dyn_coat_bk_r_06}
    trailRelaxationSpeed {editorCurveData {keys [<at -0.0134 19.3841 RCIM_Linear>,<at 0.5079 12.8986 RCIM_Linear>,<at 0.9955 9.6044 RCIM_Linear>]}}
   }
   n30 = Trail controller - Bone: dyn_coat_r_06(ComponentPose=@n29, Alpha=@n25) {
    bUsePlanarLimit=true
    chainLength=4
    debugLifeTime=0.1
    maxDeltaTime=0.0333
    alphaScaleBiasClamp {bInterpResult=true}
    rotationLimits [{limitMax=<180 180 180>}, {limitMax=<180 180 180>}, {limitMax=<180 90 180>, limitMin=<-180 -90 -180>}, {limitMax=<180 180 180>}]
    rotationOffsets [<0 0 0>, <0 0 0>, <0 0 0>, <0 0 0>]
    trailBone {boneName=dyn_coat_r_06}
    trailRelaxationSpeed {editorCurveData {keys [<at 0 15 RCIM_Linear>,<at 0.5289 13.1838 RCIM_Linear>,<at 1 6 RCIM_Linear>]}}
   }
   n31 = Anim Dynamics - Bone: dyn_sleeve_l_top_02!(ComponentPose=@n30, Alpha=1, GravityOverride:Vector="0,0,-980", ExternalForce="0,0,0") -> (Pose) {
    angularDampingOverride=0.9999
    angularSpringConstant=500
    bAngularSpring=true
    bLinearSpring=true
    bOverrideAngularDamping=true
    bUseGravityOverride=true
    bUseSphericalLimits=true
    linearSpringConstant=50
    simulationSpace=BoneRelative
    alphaBoolBlend {blendOption=Linear}
    boundBone {boneName=dyn_sleeve_l_top_02}
    componentAppliedLinearAccClamp=<1500 1500 0>
    componentLinearAccScale=<0.3 0.3 0.3>
    componentLinearVelScale=<0.3 0.3 0.3>
    gravityOverride {z=-980}
    physicsBodyDefinitions [{
      sphereCollisionRadius=0
      boundBone {boneName=dyn_sleeve_l_top_02}
      boxExtents=<5 20 20>
      constraintSetup {angularLimitsMax=<3 3 3>, angularLimitsMin=<-3 -3 -10>, angularTarget=<0 0 0>}
      localJointOffset=<3 0 0>
     }]
    relativeSpaceBone {boneName=root}
   }
   n32 = Anim Dynamics - Bone: dyn_sleeve_l_top_03!(ComponentPose=@n31, Alpha=1, GravityOverride:Vector="0,0,-980", ExternalForce="0,0,0") -> (Pose) {
    angularDampingOverride=0.9999
    angularSpringConstant=500
    bAngularSpring=true
    bLinearSpring=true
    bOverrideAngularDamping=true
    bUseGravityOverride=true
    bUseSphericalLimits=true
    linearSpringConstant=50
    simulationSpace=BoneRelative
    alphaBoolBlend {blendOption=Linear}
    boundBone {boneName=dyn_sleeve_l_top_03}
    componentAppliedLinearAccClamp=<1500 1500 0>
    componentLinearAccScale=<0.3 0.3 0.3>
    componentLinearVelScale=<0.3 0.3 0.3>
    gravityOverride {z=-980}
    physicsBodyDefinitions [{
      sphereCollisionRadius=0
      boundBone {boneName=dyn_sleeve_l_top_03}
      boxExtents=<5 20 20>
      constraintSetup {angularLimitsMax=<5 60 0>, angularLimitsMin=<-5 -60 -60>, angularTarget=<0 0 0>}
      localJointOffset=<3 0 0>
     }]
    relativeSpaceBone {boneName=root}
   }
   n33 = Anim Dynamics - Bone: dyn_sleeve_l_bk_01!(ComponentPose=@n32, Alpha=1, GravityOverride:Vector="0,0,-980", ExternalForce="0,0,0") -> (Pose) {
    angularDampingOverride=0.9999
    angularSpringConstant=500
    bAngularSpring=true
    bLinearSpring=true
    bOverrideAngularDamping=true
    bUseGravityOverride=true
    bUseSphericalLimits=true
    linearSpringConstant=50
    simulationSpace=BoneRelative
    alphaBoolBlend {blendOption=Linear}
    boundBone {boneName=dyn_sleeve_l_bk_01}
    componentAppliedLinearAccClamp=<1500 1500 0>
    componentLinearAccScale=<0.3 0.3 0.3>
    componentLinearVelScale=<0.3 0.3 0.3>
    gravityOverride {z=-980}
    physicsBodyDefinitions [{
      sphereCollisionRadius=0
      boundBone {boneName=dyn_sleeve_l_bk_01}
      boxExtents=<5 20 20>
      constraintSetup {angularLimitsMax=<2 60 5>, angularLimitsMin=<-2 -60 -60>, angularTarget=<0 0 0>}
      localJointOffset=<-3 0 0>
     }]
    relativeSpaceBone {boneName=root}
   }
   n34 = Anim Dynamics - Bone: dyn_sleeve_l_bk_02!(ComponentPose=@n33, Alpha=1, GravityOverride:Vector="0,0,-980", ExternalForce="0,0,0") -> (Pose) {
    angularDampingOverride=0.9999
    angularSpringConstant=500
    bAngularSpring=true
    bLinearSpring=true
    bOverrideAngularDamping=true
    bUseGravityOverride=true
    bUseSphericalLimits=true
    linearSpringConstant=50
    simulationSpace=BoneRelative
    alphaBoolBlend {blendOption=Linear}
    boundBone {boneName=dyn_sleeve_l_bk_02}
    componentAppliedLinearAccClamp=<1500 1500 0>
    componentLinearAccScale=<0.3 0.3 0.3>
    componentLinearVelScale=<0.3 0.3 0.3>
    gravityOverride {z=-980}
    physicsBodyDefinitions [{
      sphereCollisionRadius=0
      boundBone {boneName=dyn_sleeve_l_bk_02}
      boxExtents=<5 20 20>
      constraintSetup {angularLimitsMax=<2 60 60>, angularLimitsMin=<-2 -60 -60>, angularTarget=<0 0 0>}
      localJointOffset=<-3 0 0>
     }]
    relativeSpaceBone {boneName=root}
   }
   n35 = Anim Dynamics - Bone: dyn_sleeve_l_bot_02!(ComponentPose=@n34, Alpha=1, GravityOverride:Vector="0,0,-980", ExternalForce="0,0,0") -> (Pose) {
    angularDampingOverride=0.9999
    angularSpringConstant=500
    bAngularSpring=true
    bLinearSpring=true
    bOverrideAngularDamping=true
    bUseGravityOverride=true
    bUseSphericalLimits=true
    linearSpringConstant=50
    simulationSpace=BoneRelative
    alphaBoolBlend {blendOption=Linear}
    boundBone {boneName=dyn_sleeve_l_bot_02}
    componentAppliedLinearAccClamp=<1500 1500 0>
    componentLinearAccScale=<0.3 0.3 0.3>
    componentLinearVelScale=<0.3 0.3 0.3>
    gravityOverride {z=-980}
    physicsBodyDefinitions [{
      sphereCollisionRadius=0
      boundBone {boneName=dyn_sleeve_l_bot_02}
      boxExtents=<5 20 20>
      constraintSetup {angularLimitsMax=<3 3 3>, angularLimitsMin=<-3 -3 -10>, angularTarget=<0 0 0>}
      localJointOffset=<3 0 0>
     }]
    relativeSpaceBone {boneName=root}
   }
   n36 = Anim Dynamics - Bone: dyn_sleeve_l_bot_03!(ComponentPose=@n35, Alpha=1, GravityOverride:Vector="0,0,-980", ExternalForce="0,0,0") -> (Pose) {
    angularDampingOverride=0.9999
    angularSpringConstant=500
    bAngularSpring=true
    bLinearSpring=true
    bOverrideAngularDamping=true
    bUseGravityOverride=true
    bUseSphericalLimits=true
    linearSpringConstant=50
    simulationSpace=BoneRelative
    alphaBoolBlend {blendOption=Linear}
    boundBone {boneName=dyn_sleeve_l_bot_03}
    componentAppliedLinearAccClamp=<1500 1500 0>
    componentLinearAccScale=<0.3 0.3 0.3>
    componentLinearVelScale=<0.3 0.3 0.3>
    gravityOverride {z=-980}
    physicsBodyDefinitions [{
      sphereCollisionRadius=0
      boundBone {boneName=dyn_sleeve_l_bot_03}
      boxExtents=<5 20 20>
      constraintSetup {angularLimitsMax=<2 45 60>, angularLimitsMin=<-2 -45 -60>, angularTarget=<0 0 0>}
      localJointOffset=<3 0 0>
     }]
    relativeSpaceBone {boneName=root}
   }
   n37 = Anim Dynamics - Bone: dyn_sleeve_r_top_03!(ComponentPose=@n36, Alpha=1, GravityOverride:Vector="0,0,-980", ExternalForce="0,0,0") -> (Pose) {
    angularDampingOverride=0.9999
    angularSpringConstant=500
    bAngularSpring=true
    bLinearSpring=true
    bOverrideAngularDamping=true
    bUseGravityOverride=true
    bUseSphericalLimits=true
    linearSpringConstant=50
    simulationSpace=BoneRelative
    alphaBoolBlend {blendOption=Linear}
    boundBone {boneName=dyn_sleeve_r_top_03}
    componentAppliedLinearAccClamp=<1500 1500 0>
    componentLinearAccScale=<0.3 0.3 0.3>
    componentLinearVelScale=<0.3 0.3 0.3>
    gravityOverride {z=-980}
    physicsBodyDefinitions [{
      sphereCollisionRadius=0
      boundBone {boneName=dyn_sleeve_r_top_03}
      boxExtents=<5 20 20>
      constraintSetup {angularLimitsMax=<0 60 -5>, angularLimitsMin=<0 -60 -60>, angularTarget=<0 0 0>}
      localJointOffset=<-3 0 0>
     }]
    relativeSpaceBone {boneName=root}
   }
   n38 = Anim Dynamics - Bone: dyn_sleeve_r_bk_01!(ComponentPose=@n37, Alpha=1, GravityOverride:Vector="0,0,-980", ExternalForce="0,0,0") -> (Pose) {
    angularDampingOverride=0.9999
    angularSpringConstant=500
    bAngularSpring=true
    bLinearSpring=true
    bOverrideAngularDamping=true
    bUseGravityOverride=true
    bUseSphericalLimits=true
    linearSpringConstant=50
    simulationSpace=BoneRelative
    alphaBoolBlend {blendOption=Linear}
    boundBone {boneName=dyn_sleeve_r_bk_01}
    componentAppliedLinearAccClamp=<1500 1500 0>
    componentLinearAccScale=<0.3 0.3 0.3>
    componentLinearVelScale=<0.3 0.3 0.3>
    gravityOverride {z=-980}
    physicsBodyDefinitions [{
      sphereCollisionRadius=0
      boundBone {boneName=dyn_sleeve_r_bk_01}
      boxExtents=<5 20 20>
      constraintSetup {angularLimitsMax=<2 60 60>, angularLimitsMin=<-2 -60 -5>, angularTarget=<0 0 0>}
      localJointOffset=<-3 0 0>
     }]
    relativeSpaceBone {boneName=root}
   }
   n39 = Anim Dynamics - Bone: dyn_sleeve_r_bk_02!(ComponentPose=@n38, Alpha=1, GravityOverride:Vector="0,0,-980", ExternalForce="0,0,0") -> (Pose) {
    angularDampingOverride=0.9999
    angularSpringConstant=500
    bAngularSpring=true
    bLinearSpring=true
    bOverrideAngularDamping=true
    bUseGravityOverride=true
    bUseSphericalLimits=true
    linearSpringConstant=50
    simulationSpace=BoneRelative
    alphaBoolBlend {blendOption=Linear}
    boundBone {boneName=dyn_sleeve_r_bk_02}
    componentAppliedLinearAccClamp=<1500 1500 0>
    componentLinearAccScale=<0.3 0.3 0.3>
    componentLinearVelScale=<0.3 0.3 0.3>
    gravityOverride {z=-980}
    physicsBodyDefinitions [{
      sphereCollisionRadius=0
      boundBone {boneName=dyn_sleeve_r_bk_02}
      boxExtents=<5 20 20>
      constraintSetup {angularLimitsMax=<2 60 60>, angularLimitsMin=<-2 -60 -5>, angularTarget=<0 0 0>}
      localJointOffset=<-3 0 0>
     }]
    relativeSpaceBone {boneName=root}
   }
   n40 = Anim Dynamics - Bone: dyn_sleeve_r_bot_02!(ComponentPose=@n39, Alpha=1, GravityOverride:Vector="0,0,-980", ExternalForce="0,0,0") -> (Pose) {
    angularDampingOverride=0.9999
    angularSpringConstant=500
    bAngularSpring=true
    bLinearSpring=true
    bOverrideAngularDamping=true
    bUseGravityOverride=true
    bUseSphericalLimits=true
    linearSpringConstant=50
    simulationSpace=BoneRelative
    alphaBoolBlend {blendOption=Linear}
    boundBone {boneName=dyn_sleeve_r_bot_02}
    componentAppliedLinearAccClamp=<1500 1500 0>
    componentLinearAccScale=<0.3 0.3 0.3>
    componentLinearVelScale=<0.3 0.3 0.3>
    gravityOverride {z=-980}
    physicsBodyDefinitions [{
      sphereCollisionRadius=0
      boundBone {boneName=dyn_sleeve_r_bot_02}
      boxExtents=<5 20 20>
      constraintSetup {angularLimitsMax=<5 5 5>, angularLimitsMin=<-5 -5 -5>, angularTarget=<0 0 0>}
      localJointOffset=<-3 0 0>
     }]
    relativeSpaceBone {boneName=root}
   }
   n41 = Anim Dynamics - Bone: dyn_sleeve_r_bot_03!(ComponentPose=@n40, Alpha=1, GravityOverride:Vector="0,0,-980", ExternalForce="0,0,0") -> (Pose) {
    angularDampingOverride=0.9999
    angularSpringConstant=500
    bAngularSpring=true
    bLinearSpring=true
    bOverrideAngularDamping=true
    bUseGravityOverride=true
    bUseSphericalLimits=true
    linearSpringConstant=50
    simulationSpace=BoneRelative
    alphaBoolBlend {blendOption=Linear}
    boundBone {boneName=dyn_sleeve_r_bot_03}
    componentAppliedLinearAccClamp=<1500 1500 0>
    componentLinearAccScale=<0.3 0.3 0.3>
    componentLinearVelScale=<0.3 0.3 0.3>
    gravityOverride {z=-980}
    physicsBodyDefinitions [{
      sphereCollisionRadius=0
      boundBone {boneName=dyn_sleeve_r_bot_03}
      boxExtents=<5 20 20>
      constraintSetup {angularLimitsMax=<2 45 45>, angularLimitsMin=<-2 -45 -45>, angularTarget=<0 0 0>}
      localJointOffset=<-3 0 0>
     }]
    relativeSpaceBone {boneName=root}
   }
   Save cached pose 'trailCache'(Pose=Component To Local(ComponentPose=@n41)) {cacheName=trailCache}
   n42 = "Transform (Modify) Bone - Bone: dyn_coat_l_01"(
    ComponentPose=Local To Component(LocalPose=Use cached pose 'StartupCache'())
    Alpha=1
    Translation="0,0,0"
    Rotation="32.145,11.529,-3.902"
    Scale="1,1,1"
   ) {
    rotationMode=BMM_Additive
    rotationSpace=BCS_BoneSpace
    alphaBoolBlend {blendOption=Linear}
    boneToModify {boneName=dyn_coat_l_01}
    rotation=<rot 32.145 11.529 -3.902>
   }
   n43 = "Transform (Modify) Bone - Bone: dyn_coat_l_03"(ComponentPose=@n42, Alpha=1, Translation="0,0,0", Rotation="-0.909,17.433,8.209", Scale="1,1,1") {
    rotationMode=BMM_Additive
    rotationSpace=BCS_BoneSpace
    alphaBoolBlend {blendOption=Linear}
    boneToModify {boneName=dyn_coat_l_03}
    rotation=<rot -0.909 17.433 8.209>
   }
   n44 = "Transform (Modify) Bone - Bone: dyn_coat_l_05"(ComponentPose=@n43, Alpha=1, Translation="0,0,0", Rotation="5.741,-9.962,0.410", Scale="1,1,1") {
    rotationMode=BMM_Additive
    rotationSpace=BCS_BoneSpace
    alphaBoolBlend {blendOption=Linear}
    boneToModify {boneName=dyn_coat_l_05}
    rotation=<rot 5.741 -9.9619 0.41>
   }
   n45 = "Transform (Modify) Bone - Bone: dyn_coat_bk_l_02"(ComponentPose=@n44, Alpha=1, Translation="0,0,0", Rotation="-8.860,3.786,-1.333", Scale="1,1,1") {
    rotationMode=BMM_Additive
    rotationSpace=BCS_BoneSpace
    alphaBoolBlend {blendOption=Linear}
    boneToModify {boneName=dyn_coat_bk_l_02}
    rotation=<rot -8.86 3.786 -1.333>
   }
   n46 = "Transform (Modify) Bone - Bone: dyn_coat_bk_l_04"(ComponentPose=@n45, Alpha=1, Translation="0,0,0", Rotation="-0.703,3.196,7.134", Scale="1,1,1") {
    rotationMode=BMM_Additive
    rotationSpace=BCS_BoneSpace
    alphaBoolBlend {blendOption=Linear}
    boneToModify {boneName=dyn_coat_bk_l_04}
    rotation=<rot -0.703 3.196 7.134>
   }
   n47 = "Transform (Modify) Bone - Bone: dyn_coat_bk_l_05"(ComponentPose=@n46, Alpha=1, Translation="0,0,0", Rotation="-15.765,19.004,-5.367", Scale="1,1,1") {
    rotationMode=BMM_Additive
    rotationSpace=BCS_BoneSpace
    alphaBoolBlend {blendOption=Linear}
    boneToModify {boneName=dyn_coat_bk_l_05}
    rotation=<rot -15.765 19.004 -5.367>
   }
   n48 = "Transform (Modify) Bone - Bone: dyn_coat_bk_01"(ComponentPose=@n47, Alpha=1, Translation="0,0,0", Rotation="-15.781,-0.215,-1.367", Scale="1,1,1") {
    rotationMode=BMM_Additive
    rotationSpace=BCS_BoneSpace
    alphaBoolBlend {blendOption=Linear}
    boneToModify {boneName=dyn_coat_bk_01}
    rotation=<rot -15.781 -0.215 -1.367>
   }
   n49 = "Transform (Modify) Bone - Bone: dyn_coat_bk_02"(ComponentPose=@n48, Alpha=1, Translation="0,0,0", Rotation="-7.246,0.129,0.127", Scale="1,1,1") {
    rotationMode=BMM_Additive
    rotationSpace=BCS_BoneSpace
    alphaBoolBlend {blendOption=Linear}
    boneToModify {boneName=dyn_coat_bk_02}
    rotation=<rot -7.246 0.129 0.127>
   }
   n50 = "Transform (Modify) Bone - Bone: dyn_coat_bk_04"(ComponentPose=@n49, Alpha=1, Translation="0,0,0", Rotation="-10.235,-0.453,0.706", Scale="1,1,1") {
    rotationMode=BMM_Additive
    rotationSpace=BCS_BoneSpace
    alphaBoolBlend {blendOption=Linear}
    boneToModify {boneName=dyn_coat_bk_04}
    rotation=<rot -10.235 -0.453 0.706>
   }
   n51 = "Transform (Modify) Bone - Bone: dyn_coat_bk_05"(ComponentPose=@n50, Alpha=1, Translation="0,0,0", Rotation="-2.311,-0.443,0.100", Scale="1,1,1") {
    rotationMode=BMM_Additive
    rotationSpace=BCS_BoneSpace
    alphaBoolBlend {blendOption=Linear}
    boneToModify {boneName=dyn_coat_bk_05}
    rotation=<rot -2.311 -0.443 0.1>
   }
   n52 = "Transform (Modify) Bone - Bone: dyn_coat_bk_r_02"(ComponentPose=@n51, Alpha=1, Translation="0,0,0", Rotation="-10.873,-8.896,1.662", Scale="1,1,1") {
    rotationMode=BMM_Additive
    rotationSpace=BCS_BoneSpace
    alphaBoolBlend {blendOption=Linear}
    boneToModify {boneName=dyn_coat_bk_r_02}
    rotation=<rot -10.873 -8.896 1.662>
   }
   n53 = "Transform (Modify) Bone - Bone: dyn_coat_bk_r_04"(ComponentPose=@n52, Alpha=1, Translation="0,0,0", Rotation="-7.250,-3.668,-0.463", Scale="1,1,1") {
    rotationMode=BMM_Additive
    rotationSpace=BCS_BoneSpace
    alphaBoolBlend {blendOption=Linear}
    boneToModify {boneName=dyn_coat_bk_r_04}
    rotation=<rot -7.25 -3.668 -0.463>
   }
   n54 = "Transform (Modify) Bone - Bone: dyn_coat_bk_r_05"(ComponentPose=@n53, Alpha=1, Translation="0,0,0", Rotation="-19.163,-11.226,1.333", Scale="1,1,1") {
    rotationMode=BMM_Additive
    rotationSpace=BCS_BoneSpace
    alphaBoolBlend {blendOption=Linear}
    boneToModify {boneName=dyn_coat_bk_r_05}
    rotation=<rot -19.163 -11.226 1.333>
   }
   n55 = "Transform (Modify) Bone - Bone: dyn_coat_r_01"(ComponentPose=@n54, Alpha=1, Translation="0,0,0", Rotation="15.009,1.233,0.350", Scale="1,1,1") {
    rotationMode=BMM_Additive
    rotationSpace=BCS_BoneSpace
    alphaBoolBlend {blendOption=Linear}
    boneToModify {boneName=dyn_coat_r_01}
    rotation=<rot 15.009 1.233 0.35>
   }
   n56 = "Transform (Modify) Bone - Bone: dyn_coat_r_03"(ComponentPose=@n55, Alpha=1, Translation="0,0,0", Rotation="-21.159,-40.049,0.078", Scale="1,1,1") {
    rotationMode=BMM_Additive
    rotationSpace=BCS_BoneSpace
    alphaBoolBlend {blendOption=Linear}
    boneToModify {boneName=dyn_coat_r_03}
    rotation=<rot -21.159 -40.049 0.078>
   }
   n57 = "Transform (Modify) Bone - Bone: dyn_coat_r_05"(ComponentPose=@n56, Alpha=1, Translation="0,0,0", Rotation="0,10,0", Scale="1,1,1") {
    rotationMode=BMM_Additive
    rotationSpace=BCS_BoneSpace
    alphaBoolBlend {blendOption=Linear}
    boneToModify {boneName=dyn_coat_r_05}
    rotation {yaw=10}
   }
   n58 = Look At - Bone: dyn_coat_l_01(ComponentPose=@n57, Alpha=1) {
    bUseLookUpAxis=true
    alphaBoolBlend {blendOption=Linear}
    boneToModify {boneName=dyn_coat_l_01}
    lookAtLocation=<0.0206 -1.6234 -5.8436>
    lookAtTarget {boneReference {boneName=foot_l}}
    lookAt_Axis {axis=<1 0>}
    lookUp_Axis {axis {x=0, z=1}}
   }
   n59 = Look At - Bone: dyn_coat_r_01(ComponentPose=@n58, Alpha=1) {
    bUseLookUpAxis=true
    alphaBoolBlend {blendOption=Linear}
    boneToModify {boneName=dyn_coat_r_01}
    lookAtLocation=<7.3292 23.916 9.591>
    lookAtTarget {boneReference {boneName=foot_r}}
    lookAt_Axis {axis=<1 0>}
    lookUp_Axis {axis {x=0, z=1}}
   }
   n60 = Trail controller - Bone: dyn_coat_l_06!(ComponentPose=@n59, Alpha=1, RelaxationSpeedScale:float=2, FakeVelocity:Vector="0,0,0") -> (Pose) {
    bShowTrailLocation=true
    bUsePlanarLimit=true
    chainLength=3
    debugLifeTime=0.1
    lastBoneRotationAnimAlphaBlend=1
    maxDeltaTime=0.0333
    relaxationSpeedScale=1.2
    alphaScaleBiasClamp {bInterpResult=true, interpSpeedDecreasing=25, interpSpeedIncreasing=25}
    rotationLimits [{limitMax=<180 180 180>}, {limitMax=<180 180 180>}, {limitMax=<180 180 180>}]
    rotationOffsets [<0 0 0>, <0 0 0>, <0 0 0>]
    trailBone {boneName=dyn_coat_l_06}
    trailRelaxationSpeed {editorCurveData {keys [<at 0 15 RCIM_Linear>,<at 1 6 RCIM_Linear>]}}
   }
   n61 = Trail controller - Bone: dyn_coat_bk_l_06!(ComponentPose=@n60, Alpha=1, RelaxationSpeedScale:float=2, FakeVelocity:Vector="0,0,0") -> (Pose) {
    bShowTrailLocation=true
    bUsePlanarLimit=true
    chainLength=4
    debugLifeTime=0.1
    lastBoneRotationAnimAlphaBlend=1
    maxDeltaTime=0.0333
    relaxationSpeedScale=1.2
    alphaScaleBiasClamp {bInterpResult=true, interpSpeedDecreasing=25, interpSpeedIncreasing=25}
    rotationLimits [{limitMax=<180 180 180>}, {limitMax=<180 180 180>}, {limitMax=<180 180 180>}, {limitMax=<180 180 180>}]
    rotationOffsets [<0 0 0>, <0 0 0>, <0 0 0>, <0 0 0>]
    trailBone {boneName=dyn_coat_bk_l_06}
    trailRelaxationSpeed {editorCurveData {keys [<at 0 15 RCIM_Linear>,<at 1 6 RCIM_Linear>]}}
   }
   n62 = Trail controller - Bone: dyn_coat_bk_04!(ComponentPose=@n61, Alpha=1, RelaxationSpeedScale:float=2, FakeVelocity:Vector="0,0,0") -> (Pose) {
    bReorientParentToChild=false
    bShowTrailLocation=true
    bUsePlanarLimit=true
    debugLifeTime=0.1
    lastBoneRotationAnimAlphaBlend=1
    maxDeltaTime=0.0333
    relaxationSpeedScale=1.2
    alphaScaleBiasClamp {bInterpResult=true, interpSpeedDecreasing=25, interpSpeedIncreasing=25}
    rotationLimits [{limitMax=<180 180 180>}, {limitMax=<180 180 180>}]
    rotationOffsets [<0 0 0>, <0 0 0>]
    trailBone {boneName=dyn_coat_bk_04}
    trailRelaxationSpeed {editorCurveData {keys [<at 0 15 RCIM_Linear>,<at 1 6 RCIM_Linear>]}}
   }
   n63 = Trail controller - Bone: dyn_coat_bk_r_06!(ComponentPose=@n62, Alpha=1, RelaxationSpeedScale:float=2, FakeVelocity:Vector="0,0,0") -> (Pose) {
    bShowTrailLocation=true
    bUsePlanarLimit=true
    chainLength=3
    debugLifeTime=0.1
    lastBoneRotationAnimAlphaBlend=1
    maxDeltaTime=0.0333
    relaxationSpeedScale=1.2
    alphaScaleBiasClamp {bInterpResult=true, interpSpeedDecreasing=25, interpSpeedIncreasing=25}
    rotationLimits [{limitMax=<180 180 180>}, {limitMax=<180 180 180>}, {limitMax=<180 180 180>}]
    rotationOffsets [<0 0 0>, <0 0 0>, <0 0 0>]
    trailBone {boneName=dyn_coat_bk_r_06}
    trailRelaxationSpeed {editorCurveData {keys [<at 0 15 RCIM_Linear>,<at 1 6 RCIM_Linear>]}}
   }
   n64 = Trail controller - Bone: dyn_coat_r_06!(ComponentPose=@n63, Alpha=1, RelaxationSpeedScale:float=2, FakeVelocity:Vector="0,0,0") -> (Pose) {
    bShowTrailLocation=true
    bUsePlanarLimit=true
    chainLength=3
    debugLifeTime=0.1
    lastBoneRotationAnimAlphaBlend=1
    maxDeltaTime=0.0333
    relaxationSpeedScale=1.2
    alphaScaleBiasClamp {bInterpResult=true, interpSpeedDecreasing=25, interpSpeedIncreasing=25}
    rotationLimits [{limitMax=<180 180 180>}, {limitMax=<180 180 180>}, {limitMax=<180 180 180>}]
    rotationOffsets [<0 0 0>, <0 0 0>, <0 0 0>]
    trailBone {boneName=dyn_coat_r_06}
    trailRelaxationSpeed {editorCurveData {keys [<at 0 15 RCIM_Linear>,<at 1 6 RCIM_Linear>]}}
   }
   Save cached pose 'dbnoCache'(Pose=Component To Local(ComponentPose=@n64)) {cacheName=dbnoCache}
   n65 = Get GravityOverride()
   n66 = RigidBody!(
    ComponentPose=Local To Component(LocalPose=Use cached pose 'StartupCache'())
    Alpha
    OverrideWorldGravity:Vector=@n65
    ExternalForce
    SimSpaceSettings:Sim Space Settings Structure=(WorldAlpha=1,VelocityScaleZ=1,DampingAlpha=1,MaxLinearVelocity=1500,MaxAngularVelocity=5,MaxLinearAcceleration=500,MaxAngularAcceleration=3,ExternalLinearDragV=(X=1,Y=1,Z=0),ExternalLinearVelocity=(X=0,Y=0,Z=0),ExternalAngularVelocity=(X=0,Y=0,Z=0))
   ) -> (Pose) {
    bClampLinearTranslationLimitToRefPose=true
    bOverrideWorldGravity=true
    overridePhysicsAsset=&/F_MED_ForestJeanGrowl_Joy_Dyn_coat_Skydive
    simulationSpace=BaseBoneSpace
    alphaBoolBlend {blendOption=Linear}
    baseBoneRef {boneName=pelvis}
    overrideWorldGravity {z=-980}
    simSpaceSettings {maxAngularAcceleration=3, maxAngularVelocity=5, maxLinearAcceleration=500, maxLinearVelocity=1500, worldAlpha=1, externalLinearDragV=<1 1>}
   }
   n67 = RigidBody(
    ComponentPose=Local To Component(LocalPose=Use cached pose 'StartupCache'())
    OverrideWorldGravity=@n65
    SimSpaceSettings=(WorldAlpha=1,VelocityScaleZ=1,DampingAlpha=1,MaxLinearVelocity=1500,MaxAngularVelocity=5,MaxLinearAcceleration=500,MaxAngularAcceleration=3,ExternalLinearDragV=(X=1,Y=1,Z=0),ExternalLinearVelocity=(X=0,Y=0,Z=0),ExternalAngularVelocity=(X=0,Y=0,Z=0))
   ) {
    bClampLinearTranslationLimitToRefPose=true
    bOverrideWorldGravity=true
    overridePhysicsAsset=&/F_MED_ForestJeanGrowl_Joy_Dyn_coat_ClothOff
    simulationSpace=BaseBoneSpace
    alphaBoolBlend {blendOption=Linear}
    baseBoneRef {boneName=pelvis}
    overrideWorldGravity {z=-980}
    simSpaceSettings {maxAngularAcceleration=3, maxAngularVelocity=5, maxLinearAcceleration=500, maxLinearVelocity=1500, worldAlpha=1, externalLinearDragV=<1 1>}
   }
   n68 = Blend Poses by bool(
    BlendPose_0=Component To Local(ComponentPose=@n66)
    BlendPose_1=Component To Local(ComponentPose=@n67)
    BlendTime_0=0.2
    BlendTime_1=0.5
    bActiveValue=Get bIsParachuteOpened()
   ) {blendType=Linear}
   n69 = Use cached pose 'StartupCache'()
   n70 = Get bClothEnabled()
   n71 = RigidBody!(
    ComponentPose=Local To Component(LocalPose=Blend Poses by bool(BlendPose_0=@n69, BlendPose_1=@n68, BlendTime_0=0.2, BlendTime_1=0.5, bActiveValue=@n70) {blendType=Linear})
    Alpha
    ExternalForce
   ) -> (Pose) {
    bClampLinearTranslationLimitToRefPose=true
    bOverrideWorldGravity=true
    overridePhysicsAsset=&/F_MED_ForestJeanGrowl_Joy_Sleeve_Physics
    simulationSpace=BaseBoneSpace
    baseBoneRef {boneName=spine_05}
    componentAppliedLinearAccClamp=<3000 3000 3000>
    overrideWorldGravity {z=-980}
    simSpaceSettings {
     dampingAlpha=0
     maxAngularAcceleration=32
     maxAngularVelocity=8
     maxLinearAcceleration=2700
     maxLinearVelocity=700
     velocityScaleZ=0.3
     worldAlpha=1
     externalLinearDragV=<0.1 0.1 0.1>
    }
   }
   Save cached pose 'RbanCache'(Pose=Component To Local(ComponentPose=@n71)) {cacheName=RbanCache}
  ]}
 EventGraph {nodes [
   n1 = Event BlueprintUpdateAnimation()
   > Set DeltaTimeLocal(DeltaTimeLocal:double=@n1.DeltaTimeX) -> (Output_Get:double)
   > n2 = Sequence!(execute) -> (then_0, then_1, then_2:Exec, then_3:Exec, then_4:Exec, then_5:Exec, then_6:Exec)
   > n3 = "Gravity Override State Function (Parameter Asset) (Nativized)"(
    CharacterPartAnimInstance=Self-Reference()
    GameTimeInSeconds=GetGameTimeInSeconds()
    Parameters=&/F_MED_ForestJeanGrowl_Joy_Parameters
   )
   > Set GravityOverride(GravityOverride=@n3)
   > Set RBAN Settings Out_Native(RBAN Settings Out_Native:Fort Rigid Body Anim Node Parameters Structure=@n3.OutRigidBodyAnimNodeParameters) -> (Output_Get:Fort Rigid Body Anim Node Parameters Structure)
   n4 = "Cloth Controller Function (Parameter Asset) (Nativized)"!(
    execute=@n2.then_1
    CharacterPartAnimInstance:Custom Character Part Anim Instance Object Reference=Self-Reference()
    ClothingAssetName:Name=F_MED_ForestJeanGrowl_Joy_Clothing
    ClothCoreSettings:"Cloth Core Settings Structure (by ref)"=Get Cloth Core Settings() -> (Cloth Core Settings:Cloth Core Settings Structure)
    WindGustLocalSettings:"Wind Gust Settings Structure (by ref)"=Get Wind Gust Local Settings() -> (Wind Gust Local Settings:Wind Gust Settings Structure)
    ClothUpAxis:Vector="1,0,0"
    LinVelScaleFwdBwdClamp:float=0
    LinVelScaleUpDwnClamp:float=0.7
    DeltaTime:float=Get DeltaTimeLocal() -> (DeltaTimeLocal:double)
    GameTimeInSeconds:float=GetGameTimeInSeconds()
    StateLogicLayerName:Name=none
    Parameters:Fort Character Dynamics Parameters Object Reference=&F_MED_ForestJeanGrowl_Joy_Parameters
   ) -> (
    then
    OutCurrentState:Name=None
    OutClothParameters:Cloth Parameters Structure
    OutClothCoreSettings:Cloth Core Settings Structure
    OutWindGustSettings:Wind Gust Parameters Structure
    OutWindGustLocalSettings:Wind Gust Settings Structure
    OutWindVelocity:Vector
    OutLinearVelVector:Vector
   )
   > Set Wind Gust Local Settings(Wind Gust Local Settings:Wind Gust Settings Structure=@n4.OutWindGustLocalSettings) -> (Output_Get:Wind Gust Settings Structure)
   Set frontendAlpha(execute=@n2.then_2, frontendAlpha:double=SelectFloat(A=0.5, B=1, bPickA=Get bIsInFrontEnd())) -> (Output_Get:double)
   > Set Emote(Emote:double=SelectFloat(A=1, B=0, bPickA=Get bIsPlayingEmote())) -> (Output_Get:double)
   > Set Tactical_Sprint(Tactical_Sprint:double=SelectFloat(A=1, B=0, bPickA=Get bIsTacticalSprinting())) -> (Output_Get:double)
   n5 = Branch(execute=@n2.then_3, Condition=OR Boolean(A=Get bIsInFrontEnd(), B=Get bIsDBNOCarried()))
   > Set simSpaceChange(simSpaceChange:AnimPhysSimSpaceType Enum=World) -> (Output_Get:AnimPhysSimSpaceType Enum)
   Set simSpaceChange(execute=@n5.else)
   n6 = GetOwnerSkeletalMeshComponent()
   n7 = BreakRotator(InRot=BreakTransform(InTransform=GetDeltaTransformFromRefPose(self=@n6, BoneName=lowerarm_r, BaseName=spine_05)).Rotation)
   n8 = OR Boolean(A=Get bIsTargeting(), B=Get bIsPlayingEmote())
   Set arm_up_r(
    execute=@n2.then_4
    arm_up_r:double=SelectFloat(A=MapRangeClamped(Value=@n7.Yaw, InRangeA=45, InRangeB=20, OutRangeB=0.8), B=0, bPickA=NOT Boolean(A=@n8))
   ) -> (Output_Get:double)
   n9 = GetOwnerSkeletalMeshComponent()
   n10 = BreakRotator(InRot=BreakTransform(InTransform=GetDeltaTransformFromRefPose(self=@n9, BoneName=lowerarm_l, BaseName=spine_05)).Rotation)
   n11 = OR Boolean(A=Get bIsTargeting(), B=Get bIsPlayingEmote())
   > Set arm_up_l(arm_up_l:double=SelectFloat(A=MapRangeClamped(Value=@n10.Yaw, InRangeA=45, InRangeB=20, OutRangeB=0.5), B=0, bPickA=NOT Boolean(A=@n11))) -> (Output_Get:double)
   n12 = Branch(execute=@n2.then_5, Condition=Get bIsSwimSprinting() -> (bIsSwimSprinting:bool))
   > Set TrailControllerRot(TrailControllerRot:Rotator="15,0, 0") -> (Output_Get:Rotator)
   n13 = Branch(execute=@n12.else, Condition=Get bIsSkyDiving())
   > n14 = Branch(Condition=Get bIsParachuteOpened())
   > Set TrailControllerRot(TrailControllerRot="45,0, 0")
   n15 = Branch(execute=@n14.else, Condition="float > float"(A=Get PawnVelocityZ(), B=50) {is=title K2Node_CallFunction})
   > Set BwdMovement(BwdMovement:double=float * float(A=Get BwdMovement(), B=0)) -> (Output_Get:double)
   > Set TrailControllerRot(TrailControllerRot="45,0, 0")
   n16 = Branch(execute=@n15.else, Condition=Get bIsSkydiveDiveMode())
   > Set TrailControllerRot(TrailControllerRot="15,0, 0")
   Set TrailControllerRot(execute=@n16.else, TrailControllerRot="60,0, 0")
   Set TrailControllerRot(execute=@n13.else, TrailControllerRot="0, 0, 0")
   > Set BwdMovement(BwdMovement=float * float(A=Get BwdMovement(), B=1))
   n17 = Branch(execute=@n2.then_6, Condition=Get bIsInFrontEnd())
   > Set FE Custom Lobby Idle Is Playing Float(FE Custom Lobby Idle Is Playing Float:float=GetCurveValue(self=GetOwnerFrontEndAnimInstance(), CurveName=Frontend_Idle_Bypass)) -> (Output_Get:float)
   > n18 = Branch(Condition=Get bIsPlayingEmote())
   > Set FrontEnd Idle Alpha(FrontEnd Idle Alpha:double=0) -> (Output_Get:double)
   Set FrontEnd Idle Alpha(execute=@n18.else, FrontEnd Idle Alpha=1)
   Set FE Custom Lobby Idle Is Playing Float(execute=@n17.else, FE Custom Lobby Idle Is Playing Float=0)
   > Set FrontEnd Idle Alpha(FrontEnd Idle Alpha=0)
   n19 = Event OnCharacterCustomizationLoaded()
   > n20 = Sequence!(execute) -> (then_0, then_1, then_2)
   n21 = Branch(
    execute=@n20.then_1
    Condition=BackpackItemDefinitionHasAnyMetaTag!(Tags:Array of Gameplay Tag Structures=Make Array("[0]"=(TagName="Cosmetics.MetaTags.ForestJeanGrowl")), bRequiresEquipped:bool=false) -> (ReturnValue:bool)
   )
   > Set ForestJeanGrowl_Backpack(ForestJeanGrowl_Backpack:bool=true) -> (Output_Get:bool)
   Set ForestJeanGrowl_Backpack(execute=@n21.else, ForestJeanGrowl_Backpack=false)
   n22 = CheckForBackbling()
   Set Is Wearing Backbling(execute=@n20.then_0, Is Wearing Backbling=AND Boolean(A=@n22, B=NOT Boolean(A=@n22.bIsBackblingACape)))
   > Set Is Backbling A Cape(Is Backbling A Cape:bool=@n22.bIsBackblingACape) -> (Output_Get:bool)
   > Set Backbling(Backbling:Skeletal Mesh Component Budgeted Object Reference=Get Array Item with EFort Custom Part Type(Array=@n19.CustomMeshParts, PartType=Backpack)) -> (Output_Get:Skeletal Mesh Component Budgeted Object Reference)
   Event OnBeginSkydiving()
   > Set TrailControllerRot(TrailControllerRot="-65,0,0")
   > Set Skydive(Skydive:double=1) -> (Output_Get:double)
   Event OnEndSkydiving()
   > Set TrailControllerRot(TrailControllerRot="-10,0,0")
   > Set Skydive(Skydive=0)
   Reroute Node(InputPin:Wildcard) -> (OutputPin:Wildcard)
  ]}
]
variables [
 arm_up_l {type=number}
 arm_up_r {type=number}
 Backbling {type=SkeletalMeshComponentBudgeted}
 BwdMovement {type=number}
 Cloth Core Settings {
  type=ClothCoreSettings
  default {
   collisionThickness=0
   dampingCoefficient=0.05
   frictionCoefficient=0.3
   localDampingCoefficient=0
   numberIterations=2
   numberSubsteps=1
   selfCollisionThickness=0
   animDriveDamping=<0 1>
   animDriveStiffness=<0 1>
   areaStiffness=<1 1>
   bendingStiffness=<1 1>
   edgeStiffness=<1 1>
   tetherStiffness=<1 1>
  }
 }
 Cloth Core Settings Native {
  type=ClothCoreSettings
  default {
   collisionThickness=0
   dampingCoefficient=0.05
   frictionCoefficient=0.3
   localDampingCoefficient=0
   numberIterations=2
   numberSubsteps=1
   selfCollisionThickness=0
   animDriveDamping=<0 1>
   animDriveStiffness=<0 1>
   areaStiffness=<1 1>
   bendingStiffness=<1 1>
   edgeStiffness=<1 1>
   tetherStiffness=<1 1>
  }
 }
 DeltaTimeLocal {type=number}
 Emote {type=number}
 FE Custom Lobby Idle Is Playing Float {type=number}
 ForestJeanGrowl_Backpack {type=boolean}
 FrontEnd Idle Alpha {type=number}
 frontendAlpha {type=number}
 GravityOverride {type=Vector}
 Head_Mesh {type=SkeletalMeshComponent}
 Is Backbling A Cape {type=boolean}
 Is Wearing Backbling {type=boolean}
 RBAN Settings Out_Native {type=FortRigidBodyAnimNodeParameters}
 simSpaceChange {type=AnimPhysSimSpaceType}
 Skydive {type=number}
 Tactical_Sprint {type=number}
 TrailControllerRot {type=Rotator}
 Wind Gust Local Settings {type=WindGustSettings, default {elapsedTime=0, finalGustStrengthLocal=0, gustDuration=0, gustStrength=0, timeBetweenGusts=0}}
 Wind Gust Local Settings Native {type=WindGustSettings, default {elapsedTime=0, finalGustStrengthLocal=0, gustDuration=0, gustStrength=0, timeBetweenGusts=0}}
 Wind Gust Local Settings Native Out {type=WindGustSettings, default {elapsedTime=0, finalGustStrengthLocal=0, gustDuration=0, gustStrength=0, timeBetweenGusts=0}}
]
