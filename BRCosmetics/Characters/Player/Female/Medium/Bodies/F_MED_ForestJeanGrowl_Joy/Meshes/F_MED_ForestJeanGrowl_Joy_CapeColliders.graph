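# Physics asset for the F_MED_ForestJeanGrowl_Joy female body, providing capsule colliders that, per its name, serve the character's cape.
# Capsules cover pelvis, spine, clavicles and head, while limb and neck_02 bodies carry no shapes.
# Three back coat chains (center, left, right) have large offset capsules on their 02 and 05 bones, constrained 05 to 02 to spine_02.
# All swing and twist motions are Free, so the 45-degree limits have no effect.
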PhysicsAsset /BRCosmetics/Characters/Player/Female/Medium/Bodies/F_MED_ForestJeanGrowl_Joy/Meshes/F_MED_ForestJeanGrowl_Joy_CapeColliders
bodies {
 calf_l []
 calf_r []
 clavicle_l [{shapeType=Capsule, length=12.5623, radius=11.5194, center=<9 1.7 -3.3>, rotation=<rot -77.5566 177.3915 7.7314>}]
 clavicle_r [{shapeType=Capsule, length=12.5623, radius=11.5194, center=<-9 -1.7 3.3>, rotation=<rot -77.5566 177.3915 7.7314>}]
 dyn_coat_bk_02 [{shapeType=Capsule, length=23.4847, radius=21.4915, center=<9.4053 0 -14.1941>, rotation=<rot 90 0 0>}]
 dyn_coat_bk_05 [{shapeType=Capsule, length=22.4847, radius=22.4915, center=<22.592 0 -14.1941>, rotation=<rot 90 0 0>}]
 dyn_coat_bk_l_02 [{shapeType=Capsule, length=22.9847, radius=21.9915, center=<9.4053 0 -14.1941>, rotation=<rot 90 0 0>}]
 dyn_coat_bk_l_05 [{shapeType=Capsule, length=22.4847, radius=22.4915, center=<22.592 0 -14.1941>, rotation=<rot 90 0 0>}]
 dyn_coat_bk_r_02 [{shapeType=Capsule, length=23.4847, radius=21.4915, center=<9.4053 0 -14.1941>, rotation=<rot 90 0 0>}]
 dyn_coat_bk_r_05 [{shapeType=Capsule, length=22.4847, radius=22.4915, center=<22.592 0 -14.1941>, rotation=<rot 90 0 0>}]
 foot_l []
 foot_r []
 head [{shapeType=Capsule, length=0, radius=18, center=<9.0697 0.0746 0>, rotation=<rot 0 -90 0>}]
 lowerarm_l []
 lowerarm_r []
 neck_02 []
 pelvis [{shapeType=Capsule, length=10.1042, radius=15.6441, center=<-5.9848 -3.4909 -0.1981>}]
 spine_01 [{shapeType=Capsule, length=8.3888, radius=12.8638, center=<-3.3851 -0.5029 0>}]
 spine_02 [
  {shapeType=Capsule, length=7.0219, radius=10.7317, center=<-1.2541 -0.2891 0>}
  {shapeType=Capsule, length=20.6875, radius=9.3125, center=<-0.9903 1.9362 0>, rotation=<rot 90 0 0>}
 ]
 spine_03 [
  {shapeType=Capsule, length=5.6863, radius=11.4854, center=<0.0574 0.4416 0>}
  {shapeType=Capsule, length=20.7188, radius=9.2812, center=<-1.2163 0.549 0>, rotation=<rot 90 180 -175.0732>}
 ]
 spine_04 [
  {shapeType=Capsule, length=9.5076, radius=10.8916, center=<3.2612 -0.6779 0>}
  {shapeType=Capsule, length=10.0312, radius=9.2812, center=<-1.2326 1.308 0>, rotation=<rot 90 0 0>}
 ]
 spine_05 [{shapeType=Capsule, length=29.678, radius=12.3571, center=<0.6969 0.7637 0>}]
 thigh_l []
 thigh_r []
 upperarm_l []
 upperarm_r []
}
constraints [
 {
  bone1Name=dyn_coat_bk_l_02
  bone2Name=spine_02
  swing1LimitDegrees=45
  swing1Motion=Free
  swing2LimitDegrees=45
  swing2Motion=Free
  twistLimitDegrees=45
  twistMotion=Free
 }
 {
  bone1Name=dyn_coat_bk_02
  bone2Name=spine_02
  swing1LimitDegrees=45
  swing1Motion=Free
  swing2LimitDegrees=45
  swing2Motion=Free
  twistLimitDegrees=45
  twistMotion=Free
 }
 {
  bone1Name=dyn_coat_bk_r_02
  bone2Name=spine_02
  swing1LimitDegrees=45
  swing1Motion=Free
  swing2LimitDegrees=45
  swing2Motion=Free
  twistLimitDegrees=45
  twistMotion=Free
 }
 {
  bone1Name=dyn_coat_bk_r_05
  bone2Name=dyn_coat_bk_r_02
  swing1LimitDegrees=45
  swing1Motion=Free
  swing2LimitDegrees=45
  swing2Motion=Free
  twistLimitDegrees=45
  twistMotion=Free
 }
 {
  bone1Name=dyn_coat_bk_l_05
  bone2Name=dyn_coat_bk_l_02
  swing1LimitDegrees=45
  swing1Motion=Free
  swing2LimitDegrees=45
  swing2Motion=Free
  twistLimitDegrees=45
  twistMotion=Free
 }
 {
  bone1Name=dyn_coat_bk_05
  bone2Name=dyn_coat_bk_02
  swing1LimitDegrees=45
  swing1Motion=Free
  swing2LimitDegrees=45
  swing2Motion=Free
  twistLimitDegrees=45
  twistMotion=Free
 }
]
