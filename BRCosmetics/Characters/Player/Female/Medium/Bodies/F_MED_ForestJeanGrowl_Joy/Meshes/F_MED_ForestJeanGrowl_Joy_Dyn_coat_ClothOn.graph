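# Physics asset for the dynamic coat bones on the ForestJeanGrowl_Joy medium female player body.
# Left and right three-capsule coat chains hang from small MSTR bodies, with tight twist and swing1 limits narrowing down each chain.
# Pelvis, spine and thigh capsules form the body volumes, linked by free-rotation constraints.
# Each MSTR is constrained to both spine_03 and spine_04, though spine_04 has no body here.
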PhysicsAsset /BRCosmetics/Characters/Player/Female/Medium/Bodies/F_MED_ForestJeanGrowl_Joy/Meshes/F_MED_ForestJeanGrowl_Joy_Dyn_coat_ClothOn
bodies {
 dyn_coat_l_01 [{shapeType=Capsule, length=10.7561, radius=8, shapeName=dyn_coat_l_01_capsule, center=<4.667 0 6.211>, rotation=<rot 90 0 0>}]
 dyn_coat_l_02 [{shapeType=Capsule, length=12.2561, radius=8, shapeName=dyn_coat_l_01_capsule, center=<5.44 0 5.5818>, rotation=<rot 90 0 0>}]
 dyn_coat_l_03 [{shapeType=Capsule, length=21.2561, radius=8, shapeName=dyn_coat_l_01_capsule, center=<10.3356 0 5.0234>, rotation=<rot 90 0 0>}]
 dyn_coat_l_MSTR [{shapeType=Capsule, length=0.505, radius=0.505, shapeName=dyn_coat_l_MSTR_capsule}]
 dyn_coat_r_01 [{shapeType=Capsule, length=10.7561, radius=8, shapeName=dyn_coat_l_01_capsule, center=<4.667 0 6.211>, rotation=<rot 90 0 0>}]
 dyn_coat_r_02 [{shapeType=Capsule, length=12.2561, radius=8, shapeName=dyn_coat_l_01_capsule, center=<5.44 0 5.5818>, rotation=<rot 90 0 0>}]
 dyn_coat_r_03 [{shapeType=Capsule, length=21.2561, radius=8, shapeName=dyn_coat_l_01_capsule, center=<10.3356 0 5.0234>, rotation=<rot 90 0 0>}]
 dyn_coat_r_MSTR [{shapeType=Capsule, length=0.505, radius=0.505, shapeName=dyn_coat_l_MSTR_capsule}]
 pelvis [{shapeType=Capsule, length=6.467, radius=13.521, shapeName=pelvis_capsule, center=<-4.8171 -0.6518 -0.0005>}]
 spine_01 [{shapeType=Capsule, length=4.4652, radius=10.9987, shapeName=pelvis_capsule, center=<-0.9991 -0.3721 0>}]
 spine_02 [{shapeType=Capsule, length=1.4401, radius=10.4987, shapeName=pelvis_capsule, center=<0.1451 0.5237 0>}]
 spine_03 [{shapeType=Capsule, length=2.9401, radius=10.4987, shapeName=pelvis_capsule, center=<2.1769 2.6568 0>}]
 thigh_l [{
   shapeType=Capsule
   length=29.589
   radius=11.0541
   shapeName=pelvis_capsule
   center=<-20.0609 -2.1741 1.8369>
   rotation=<rot 88.6498 0.0001 0.0002>
  }]
 thigh_r [{
   shapeType=Capsule
   length=29.589
   radius=11.0541
   shapeName=pelvis_capsule
   center=<20.0636 2.1741 -1.7232>
   rotation=<rot 88.6498 0.0001 -179.9998>
  }]
}
constraints [
 {bone1Name=dyn_coat_l_01, bone2Name=dyn_coat_l_MSTR, swing1LimitDegrees=45, swing2LimitDegrees=45, twistLimitDegrees=5}
 {bone1Name=dyn_coat_l_02, bone2Name=dyn_coat_l_01, swing1LimitDegrees=25, swing2LimitDegrees=45, twistLimitDegrees=5}
 {bone1Name=dyn_coat_l_03, bone2Name=dyn_coat_l_02, swing1LimitDegrees=10, swing2LimitDegrees=45, twistLimitDegrees=5}
 {bone1Name=dyn_coat_r_01, bone2Name=dyn_coat_r_MSTR, swing1LimitDegrees=45, swing2LimitDegrees=45, twistLimitDegrees=5}
 {bone1Name=dyn_coat_r_02, bone2Name=dyn_coat_r_01, swing1LimitDegrees=25, swing2LimitDegrees=45, twistLimitDegrees=5}
 {bone1Name=dyn_coat_r_03, bone2Name=dyn_coat_r_02, swing1LimitDegrees=10, swing2LimitDegrees=45, twistLimitDegrees=5}
 {
  bone1Name=dyn_coat_l_MSTR
  bone2Name=spine_03
  swing1LimitDegrees=45
  swing1Motion=Free
  swing2LimitDegrees=45
  swing2Motion=Free
  twistLimitDegrees=45
  twistMotion=Free
 }
 {
  bone1Name=dyn_coat_r_MSTR
  bone2Name=spine_03
  swing1LimitDegrees=45
  swing1Motion=Free
  swing2LimitDegrees=45
  swing2Motion=Free
  twistLimitDegrees=45
  twistMotion=Free
 }
 {
  bone1Name=spine_02
  bone2Name=spine_01
  swing1LimitDegrees=45
  swing1Motion=Free
  swing2LimitDegrees=45
  swing2Motion=Free
  twistLimitDegrees=45
  twistMotion=Free
 }
 {
  bone1Name=spine_03
  bone2Name=spine_02
  swing1LimitDegrees=45
  swing1Motion=Free
  swing2LimitDegrees=45
  swing2Motion=Free
  twistLimitDegrees=45
  twistMotion=Free
 }
 {
  bone1Name=dyn_coat_l_MSTR
  bone2Name=spine_04
  swing1LimitDegrees=45
  swing1Motion=Free
  swing2LimitDegrees=45
  swing2Motion=Free
  twistLimitDegrees=45
  twistMotion=Free
 }
 {
  bone1Name=dyn_coat_r_MSTR
  bone2Name=spine_04
  swing1LimitDegrees=45
  swing1Motion=Free
  swing2LimitDegrees=45
  swing2Motion=Free
  twistLimitDegrees=45
  twistMotion=Free
 }
 {
  bone1Name=spine_01
  bone2Name=pelvis
  swing1LimitDegrees=45
  swing1Motion=Free
  swing2LimitDegrees=45
  swing2Motion=Free
  twistLimitDegrees=45
  twistMotion=Free
 }
 {
  bone1Name=thigh_l
  bone2Name=pelvis
  swing1LimitDegrees=45
  swing1Motion=Free
  swing2LimitDegrees=45
  swing2Motion=Free
  twistLimitDegrees=45
  twistMotion=Free
 }
 {
  bone1Name=thigh_r
  bone2Name=pelvis
  swing1LimitDegrees=45
  swing1Motion=Free
  swing2LimitDegrees=45
  swing2Motion=Free
  twistLimitDegrees=45
  twistMotion=Free
 }
]
